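# Material function for the BlondeJaw Strike character body that adds damage and elimination reactive glow, adjusting emissive, base color and roughness.
# Five damage and two elimination start-time slots drive ramp envelopes, compressed by S-curves into thresholds along a gamma-shaped gradient, weighted by glow masks.
# Curve atlas rows color the result, broken up by animated noise on stabilized screen-space UVs.
# A StrikeCard switch selects variant base emissive and masking; debug switches override reactivity.
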MaterialFunction /BRCosmetics/Characters/Player/Male/Medium/Bodies/M_MED_BlondeJaw_Strike/Materials/MF_BlondeJaw
graph {nodes [
  Constant()
  FunctionInput() {inputName=BaseEmissiveMask, inputType=FunctionInput_Scalar, sortPriority=40}
  n1 = GetMaterialAttributes(None=FunctionInput() {inputType=FunctionInput_MaterialAttributes, sortPriority=-1}) {attributeGetTypes [#1, #2, #3]}
  n2 = ScalarParameter() {group=BlondeJaw, parameterName=BaseEmissiveOverride_MaskThreshold, sliderMax=0.999, sortPriority=31}
  n3 = Subtract(A=NamedRerouteUsage(), B=@n2)
  n4 = Max(A=OneMinus(None=@n2)) {constB=0.001}
  n5 = Power(
   Base=Saturate(None=Divide(A=@n3, B=@n4))
   Exp=ScalarParameter() {defaultValue=1, group=BlondeJaw, parameterName=BaseEmissiveOverride_Exp, sortPriority=31}
  )
  n6 = ScalarParameter() {group=BlondeJaw, parameterName=BaseEmissiveOverride_Bias, sortPriority=31}
  n7 = CurveAtlasRowParameter(CurveTime=Saturate(None=Subtract(A=@n5, B=@n6))) {
   atlas=&/CA_BlondeJaw
   curve=&/C_BlondeJaw_Strike
   group=BlondeJaw
   parameterName=Reactivity_EmissiveCurve
  }
  n8 = Desaturation(
   None=Multiply(A=@n5, B=@n7)
   Fraction=ScalarParameter() {group=BlondeJaw, parameterName=BaseEmissiveOverride_Desaturation, sliderMax=1, sortPriority=31}
  )
  n9 = StaticSwitchParameter(
   True=Multiply(A=@n8, B=ScalarParameter() {defaultValue=1, group=BlondeJaw, parameterName=BaseEmissiveOverride_Brightness, sortPriority=31})
   False=Multiply(A=ScalarParameter() {group=BlondeJaw, parameterName=EnableBaseEmissive, sortPriority=30}, B=@n1.EmissiveColor)
  ) {group=BlondeJaw, parameterName=StrikeCard, sortPriority=30}
  n10 = NamedRerouteUsage()
  n11 = Multiply(A=ScalarParameter() {defaultValue=3, group=BlondeJaw, parameterName=ElimReactivity_Brightness}, B=@n10)
  n12 = ScalarParameter() {defaultValue=1, group=BlondeJaw, parameterName=LowFreqBreakup_Scale}
  n13 = MaterialFunctionCall!(Position, UVRadialScale) {materialFunction=/Game/Characters/Player/Common/MaterialFunctions/Superlevel/CharacterStabilizedScreenSpaceUVs}
  n14 = Constant2Vector() {g=0.3, r=0.09}
  n15 = ScalarParameter() {defaultValue=1, group=BlondeJaw, parameterName=LowFreqBreakup_Speed}
  n16 = TextureSample(UVs=Panner(Coordinate=Multiply(A=@n12, B=@n13.StabilizedUVs), Speed=Multiply(A=@n14, B=@n15)) {speedX=0.09, speedY=0.3}) {samplerType=SAMPLERTYPE_Grayscale, texture=/Game/Global/Textures/Default/Noise/LowResBlurredNoiseNormalized}
  n17 = LinearInterpolate(A=@n16.R, Alpha=ScalarParameter() {defaultValue=0.8, group=BlondeJaw, parameterName=LowFreqBreakup_MixAlpha, sliderMax=1}) {constAlpha=0.9}
  n18 = Multiply(A=@n10, B=@n17)
  n19 = CurveAtlasRowParameter(CurveTime=@n18) {
   atlas=&CA_BlondeJaw
   curve=&C_BlondeJaw_Strike
   group=BlondeJaw
   parameterName=ElimReactivity_EmissiveCurve
  }
  n20 = CurveAtlasRowParameter(CurveTime=@n18) {
   atlas=&CA_BlondeJaw
   curve=&C_BlondeJaw_Strike
   group=BlondeJaw
   parameterName=Reactivity_EmissiveCurve
  }
  n21 = StaticSwitchParameter(True=@n19, False=@n20) {group=BlondeJaw, parameterName=ElimReactivity_UseSeparateCurve}
  n22 = NamedRerouteUsage()
  n23 = Multiply(
   A=ScalarParameter() {defaultValue=3, group=BlondeJaw, parameterName=DmgReactivity_Brightness}
   B=Power(Base=@n22, Exp=ScalarParameter() {defaultValue=2, group=BlondeJaw, parameterName=Reactivity_BrightnessExp})
  )
  n24 = ScalarParameter() {defaultValue=1, group=BlondeJaw, parameterName=NoiseBreakupScale}
  n25 = TextureObjectParameter() {
   group=BlondeJaw
   parameterName=EnergyNoise
   texture=/BRCosmetics/Characters/Player/Male/Medium/Bodies/M_MED_BlondeJaw_Strike/Textures/T_BlondeJaw_NoiseBreakup_M
  }
  n26 = ComponentMask(None=MaterialFunctionCall!(
    Coordinates=Multiply(A=@n13.StabilizedUVs, B=@n24) {constB=150}
    Speed=ScalarParameter() {defaultValue=1, group=BlondeJaw, parameterName=NoiseBreakupAnimSpeed}
    Divisor
    Texture=@n25
   ) {materialFunction=/Engine/Functions/Engine_MaterialFunctions03/ComplexMotion/Motion_4WayChaos}.Result) {r=true}
  n27 = Multiply(A=Power(Base=@n26), B=ScalarParameter() {defaultValue=20, group=BlondeJaw, parameterName=NoiseBreakup_Intensity}) {constB=20}
  n28 = CurveAtlasRowParameter(CurveTime=Multiply(A=Multiply(A=@n22, B=@n17), B=LinearInterpolate(Alpha=@n27) {constA=0.5})) {
   atlas=&CA_BlondeJaw
   curve=&C_BlondeJaw_Strike
   group=BlondeJaw
   parameterName=Reactivity_EmissiveCurve
  }
  n29 = LinearInterpolate(
   A=@n28
   B=Multiply(A=VectorParameter() {group=BlondeJaw, parameterName=LowFreqBreakup_Tint, defaultValue=<rgba 0 1 0 1>}.RGB, B=@n28)
   Alpha=@n16.R
  )
  n30 = CurveAtlasRowParameter(CurveTime=@n27) {
   atlas=&CA_BlondeJaw
   curve=&C_BlondeJaw_Strike
   group=BlondeJaw
   parameterName=Reactivity_EmissiveCurve
  }
  n31 = NamedRerouteUsage()
  n32 = ScalarParameter() {defaultValue=2, group=BlondeJaw, parameterName=NoiseBreakup_GlowMaskExponent}
  n33 = Multiply(
   A=Multiply(A=@n30, B=@n26)
   B=LinearInterpolate(
    B=Power(Base=@n31, Exp=@n32)
    Alpha=ScalarParameter() {defaultValue=1, group=BlondeJaw, parameterName=NoiseBreakup_GlowMaskInfluence, sliderMax=1}
   ) {constA=1}
  )
  n34 = Power(
   Base=SphereMask(
    A=NamedRerouteUsage()
    B=ScalarParameter() {defaultValue=1, group=BlondeJaw, parameterName=NoiseBreakup_MaskCenter}
    Radius=ScalarParameter() {defaultValue=0.3, group=BlondeJaw, parameterName=NoiseBreakup_MaskWidth}
   ) {attenuationRadius=0.2, hardnessPercent=0}
   Exp=ScalarParameter() {defaultValue=1, group=BlondeJaw, parameterName=NoiseBreakup_MaskExp}
  )
  n35 = NamedRerouteUsage()
  n36 = ScalarParameter() {defaultValue=5, group=BlondeJaw, parameterName=NoiseBreakup_DimCurve}
  n37 = Multiply(A=Multiply(A=@n33, B=@n34), B=Saturate(None=Power(Base=@n35, Exp=@n36)))
  n38 = OneMinus(None=NamedRerouteUsage())
  n39 = Add(A=NamedRerouteUsage(), B=ScalarParameter() {group=BlondeJaw, parameterName=NoiseBreakup_ReactivityWake_Bias, sliderMax=1, sliderMin=-1})
  n40 = ScalarParameter() {defaultValue=0.2, group=BlondeJaw, parameterName=NoiseBreakup_ReactivityWake_Width, sliderMax=1}
  n41 = Multiply(A=Multiply(A=@n33, B=@n34), B=Multiply(A=SphereMask(A=@n38, B=@n39, Radius=@n40) {hardnessPercent=0}, B=NamedRerouteUsage()))
  n42 = Saturate(None=Multiply(A=NamedRerouteUsage(), B=ScalarParameter() {defaultValue=20, group=BlondeJaw, parameterName=NoiseBreakup_DimSteepness_StrikeCard}))
  n43 = StaticSwitchParameter(True=Max(A=@n37, B=Multiply(A=@n41, B=@n42)), False=@n37) {group=BlondeJaw, parameterName=StrikeCard, sortPriority=30}
  n44 = VectorParameter() {group=BlondeJaw, parameterName=NoiseBreakup_ColorTint, defaultValue=<rgba 0.3042 0.026 1 1>}
  n45 = Add(
   A=Add(A=Multiply(A=@n11, B=@n21), B=Multiply(A=@n23, B=@n29))
   B=Multiply(A=Multiply(A=@n43, B=@n44.RGB), B=ScalarParameter() {defaultValue=5, group=BlondeJaw, parameterName=NoiseBreakupBrightness}) {constB=5}
  )
  n46 = StaticSwitchParameter(True=LinearInterpolate(A=@n1.BaseColor, B=@n8, Alpha=NamedRerouteUsage()), False=@n1.BaseColor) {group=BlondeJaw, parameterName=StrikeCard, sortPriority=30}
  n47 = NamedRerouteUsage()
  FunctionOutput(None=SetMaterialAttributes!(
    MaterialAttributes=@n1.MaterialAttributes
    EmissiveColor=Add(A=@n9, B=@n45)
    BaseColor=LinearInterpolate(A=@n46, Alpha=@n47) {constB=0}
    Roughness=LinearInterpolate(A=@n1.Roughness, Alpha=@n47)
   ) {attributeSetTypes [#1, #2, #3]}) {outputName=SetMaterialAttributes}
  n48 = ScalarParameter() {defaultValue=1, group=BlondeJaw, parameterName=DebugReactivity_NumHits, sliderMax=5, sortPriority=31}
  n49 = ScalarParameter() {defaultValue=1.3, group=BlondeJaw, parameterName=ReactivitySingleHitStrength}
  n50 = Saturate(None=@n48)
  n51 = ScalarParameter() {defaultValue=0.4, group=BlondeJaw, parameterName=ReactivityFirstHitStrengthBoost}
  n52 = MaterialFunctionCall!(In=Add(A=Multiply(A=@n48, B=@n49), B=Multiply(A=@n50, B=@n51) {constB=0.4})) {materialFunction=/Game/Global/MaterialFunctions/MF_SCurveFunctions}
  LinearInterpolate(
   A=@n52."Logistic Function (-1...1)"
   B=@n52."Logistic Function (-1...1)"
   Alpha=ScalarParameter() {group=BlondeJaw, parameterName=DebugReactivity_UseTanH, sliderMax=1, sortPriority=31}
  )
  n53 = Subtract(A=Time(), B=ScalarParameter() {defaultValue=-100, group=BlondeJaw, parameterName=ElimReactivityStartTime_0})
  n54 = ScalarParameter() {defaultValue=1, group=BlondeJaw, parameterName=ElimReactivityRampUpDuration}
  n55 = Power(
   Base=Saturate(None=Divide(A=@n53, B=@n54) {constB=0.8})
   Exp=ScalarParameter() {defaultValue=1, group=BlondeJaw, parameterName=ElimReactivityRampUpExp}
  ) {constExponent=0.8}
  n56 = Divide(A=@n53, B=ScalarParameter() {defaultValue=5, group=BlondeJaw, parameterName=ElimReactivityRampDownDuration}) {constB=0.8}
  n57 = Power(Base=OneMinus(None=Saturate(None=@n56)), Exp=ScalarParameter() {defaultValue=2, group=BlondeJaw, parameterName=ElimReactivityRampDownExp}) {constExponent=0.4}
  n58 = Subtract(A=Time(), B=ScalarParameter() {defaultValue=-100, group=BlondeJaw, parameterName=ElimReactivityStartTime_1})
  n59 = ScalarParameter() {defaultValue=1, group=BlondeJaw, parameterName=ElimReactivityRampUpDuration}
  n60 = Power(
   Base=Saturate(None=Divide(A=@n58, B=@n59) {constB=0.8})
   Exp=ScalarParameter() {defaultValue=1, group=BlondeJaw, parameterName=ElimReactivityRampUpExp}
  ) {constExponent=0.8}
  n61 = Divide(A=@n58, B=ScalarParameter() {defaultValue=5, group=BlondeJaw, parameterName=ElimReactivityRampDownDuration}) {constB=0.8}
  n62 = Power(Base=OneMinus(None=Saturate(None=@n61)), Exp=ScalarParameter() {defaultValue=2, group=BlondeJaw, parameterName=ElimReactivityRampDownExp}) {constExponent=0.4}
  n63 = Add(
   A=Multiply(A=Multiply(A=@n55, B=@n57), B=ScalarParameter() {defaultValue=0.7, group=BlondeJaw, parameterName=ElimReactivitySingleHitStrength})
   B=Multiply(A=Multiply(A=@n60, B=@n62), B=ScalarParameter() {defaultValue=0.7, group=BlondeJaw, parameterName=ElimReactivitySingleHitStrength})
  )
  Multiply(A=@n63)
  NamedRerouteDeclaration(None=FunctionInput() {inputName=GlowMask, inputType=FunctionInput_Scalar, sortPriority=20}) {name=GlowMask, nodeColor=<rgba 0 1 0.0156 1>}
  n64 = OneMinus(None=FunctionInput() {inputName=ReactivityGradient, inputType=FunctionInput_Scalar, sortPriority=30})
  n65 = ScalarParameter() {defaultValue=1.8, group=BlondeJaw, parameterName=Reactivity_GradientGamma, sliderMax=2, sliderMin=0.5}
  NamedRerouteDeclaration(None=OneMinus(None=Power(Base=@n64, Exp=@n65))) {name=ReactivityGradient, nodeColor=<rgba 0 1 0.0156 1>}
  n66 = NamedRerouteUsage()
  n67 = Time()
  n68 = StaticSwitchParameter(
   True=Frac(None=Multiply(A=@n67) {constB=0.2})
   False=StaticSwitchParameter(
    True=@n52."Logistic Function (-1...1)"
    False=ScalarParameter() {defaultValue=1, group=BlondeJaw, parameterName=DebugReactivity_DirectAlpha, sliderMax=1, sortPriority=31}
   ) {group=BlondeJaw, parameterName=DebugReactivity_TestRawHitCount, sortPriority=30}
  ) {group=BlondeJaw, parameterName=DebugReactivity_Loop, sortPriority=30}
  n69 = ScalarParameter() {defaultValue=1.3, group=BlondeJaw, parameterName=ReactivitySingleHitStrength}
  n70 = Subtract(A=Time(), B=ScalarParameter() {defaultValue=-100, group=BlondeJaw, parameterName=DmgReactivityStartTime_0})
  n71 = Saturate(None=Divide(A=@n70, B=ScalarParameter() {defaultValue=0.3, group=BlondeJaw, parameterName=ReactivityRampUpDuration}) {constB=0.8})
  n72 = ScalarParameter() {defaultValue=0.8, group=BlondeJaw, parameterName=ReactivityRampUpExp}
  n73 = ScalarParameter() {defaultValue=10, group=BlondeJaw, parameterName=ReactivityRampDownDuration}
  n74 = OneMinus(None=Saturate(None=Divide(A=@n70, B=@n73) {constB=0.8}))
  n75 = ScalarParameter() {defaultValue=2, group=BlondeJaw, parameterName=ReactivityRampDownExp}
  n76 = Subtract(A=Time(), B=ScalarParameter() {defaultValue=-100, group=BlondeJaw, parameterName=DmgReactivityStartTime_1})
  n77 = ScalarParameter() {defaultValue=0.3, group=BlondeJaw, parameterName=ReactivityRampUpDuration}
  n78 = Power(
   Base=Saturate(None=Divide(A=@n76, B=@n77) {constB=0.8})
   Exp=ScalarParameter() {defaultValue=0.8, group=BlondeJaw, parameterName=ReactivityRampUpExp}
  ) {constExponent=0.8}
  n79 = Divide(A=@n76, B=ScalarParameter() {defaultValue=10, group=BlondeJaw, parameterName=ReactivityRampDownDuration}) {constB=0.8}
  n80 = Power(Base=OneMinus(None=Saturate(None=@n79)), Exp=ScalarParameter() {defaultValue=2, group=BlondeJaw, parameterName=ReactivityRampDownExp}) {constExponent=0.4}
  n81 = Subtract(A=Time(), B=ScalarParameter() {defaultValue=-100, group=BlondeJaw, parameterName=DmgReactivityStartTime_2})
  n82 = Divide(A=@n81, B=ScalarParameter() {defaultValue=0.3, group=BlondeJaw, parameterName=ReactivityRampUpDuration}) {constB=0.8}
  n83 = Saturate(None=Divide(A=@n81, B=ScalarParameter() {defaultValue=10, group=BlondeJaw, parameterName=ReactivityRampDownDuration}) {constB=0.8})
  n84 = Multiply(
   A=Power(Base=Saturate(None=@n82), Exp=ScalarParameter() {defaultValue=0.8, group=BlondeJaw, parameterName=ReactivityRampUpExp}) {constExponent=0.8}
   B=Power(Base=OneMinus(None=@n83), Exp=ScalarParameter() {defaultValue=2, group=BlondeJaw, parameterName=ReactivityRampDownExp}) {constExponent=0.4}
  )
  n85 = Subtract(A=Time(), B=ScalarParameter() {defaultValue=-100, group=BlondeJaw, parameterName=DmgReactivityStartTime_3})
  n86 = Saturate(None=Divide(A=@n85, B=ScalarParameter() {defaultValue=0.3, group=BlondeJaw, parameterName=ReactivityRampUpDuration}) {constB=0.8})
  n87 = ScalarParameter() {defaultValue=0.8, group=BlondeJaw, parameterName=ReactivityRampUpExp}
  n88 = ScalarParameter() {defaultValue=10, group=BlondeJaw, parameterName=ReactivityRampDownDuration}
  n89 = OneMinus(None=Saturate(None=Divide(A=@n85, B=@n88) {constB=0.8}))
  n90 = ScalarParameter() {defaultValue=2, group=BlondeJaw, parameterName=ReactivityRampDownExp}
  n91 = Subtract(A=Time(), B=ScalarParameter() {defaultValue=-100, group=BlondeJaw, parameterName=DmgReactivityStartTime_4})
  n92 = Saturate(None=Divide(A=@n91, B=ScalarParameter() {defaultValue=0.3, group=BlondeJaw, parameterName=ReactivityRampUpDuration}) {constB=0.8})
  n93 = ScalarParameter() {defaultValue=0.8, group=BlondeJaw, parameterName=ReactivityRampUpExp}
  n94 = ScalarParameter() {defaultValue=10, group=BlondeJaw, parameterName=ReactivityRampDownDuration}
  n95 = OneMinus(None=Saturate(None=Divide(A=@n91, B=@n94) {constB=0.8}))
  n96 = ScalarParameter() {defaultValue=2, group=BlondeJaw, parameterName=ReactivityRampDownExp}
  n97 = Add(
   A=Multiply(A=Power(Base=@n86, Exp=@n87) {constExponent=0.8}, B=Power(Base=@n89, Exp=@n90) {constExponent=0.4})
   B=Multiply(A=Power(Base=@n92, Exp=@n93) {constExponent=0.8}, B=Power(Base=@n95, Exp=@n96) {constExponent=0.4})
  )
  n98 = Add(
   A=Multiply(A=Power(Base=@n71, Exp=@n72) {constExponent=0.8}, B=Power(Base=@n74, Exp=@n75) {constExponent=0.4})
   B=Add(A=Multiply(A=@n78, B=@n80), B=Add(A=@n84, B=@n97))
  )
  n99 = Saturate(None=@n98)
  n100 = ScalarParameter() {defaultValue=0.4, group=BlondeJaw, parameterName=ReactivityFirstHitStrengthBoost}
  n101 = MaterialFunctionCall(In=Add(A=Multiply(A=@n98, B=@n69), B=Multiply(A=@n99, B=@n100) {constB=0.4})) {materialFunction=&MF_SCurveFunctions}
  n102 = ConstantBiasScale(None=NamedRerouteDeclaration(None=StaticSwitchParameter(True=@n68, False=@n101."Logistic Function (-1...1)") {group=BlondeJaw, parameterName=DebugReactivity, sortPriority=30}) {name=RawReactivityPercentage, nodeColor=<rgba 0.8516 1 0 1>}) {bias=-0.5, scale=-2}
  n103 = Subtract(A=Saturate(None=Subtract(A=@n66, B=@n102))) {constB=0.5}
  n104 = ScalarParameter() {defaultValue=3, group=BlondeJaw, parameterName=Reactivity_GradientSteepness}
  n105 = Saturate(None=Add(A=Multiply(A=@n103, B=@n104)) {constB=0.5})
  n106 = NamedRerouteUsage()
  n107 = StaticSwitchParameter(True=Power(Base=@n106, Exp=ScalarParameter() {defaultValue=1, group=BlondeJaw, parameterName=DmgReactivity_GlowMask_Exp}), False=@n106) {group=BlondeJaw, parameterName=StrikeCard, sortPriority=30}
  NamedRerouteDeclaration(None=Multiply(
    A=Multiply(A=@n105, B=@n107)
    B=ScalarParameter() {defaultValue=1, group=BlondeJaw, parameterName=EnableReactivity, sliderMax=1, sortPriority=-100}
   )) {name=DmgReactivityProgress, nodeColor=<rgba 0 0.625 1 1>}
  n108 = NamedRerouteUsage()
  n109 = ScalarParameter() {defaultValue=1, group=BlondeJaw, parameterName=DebugElimReactivity_DirectAlpha, sliderMax=1, sortPriority=31}
  n110 = MaterialFunctionCall(In=@n63) {materialFunction=&MF_SCurveFunctions}
  n111 = ConstantBiasScale(None=NamedRerouteDeclaration(None=StaticSwitchParameter(True=@n109, False=@n110."Tanh Function (-1...1)") {group=BlondeJaw, parameterName=DebugReactivity, sortPriority=30}) {name=RawElimReactivityPercentage, nodeColor=<rgba 0.5234 0 1 1>}) {bias=-0.5, scale=-2}
  n112 = Subtract(A=Saturate(None=Subtract(A=@n108, B=@n111))) {constB=0.5}
  n113 = ScalarParameter() {defaultValue=3, group=BlondeJaw, parameterName=Reactivity_GradientSteepness}
  n114 = Saturate(None=Add(A=Multiply(A=@n112, B=@n113)) {constB=0.5})
  n115 = Power(Base=NamedRerouteUsage(), Exp=ScalarParameter() {defaultValue=2, group=BlondeJaw, parameterName=ElimReactivity_GlowMask_Exp})
  NamedRerouteDeclaration(None=Multiply(
    A=Multiply(A=@n114, B=@n115)
    B=ScalarParameter() {defaultValue=1, group=BlondeJaw, parameterName=EnableReactivity, sliderMax=1, sortPriority=-100}
   )) {name=ElimReactivityProgress, nodeColor=<rgba 0 1 0.8359 1>}
  NamedRerouteDeclaration(None=Desaturation(None=NamedRerouteUsage())) {name=BaseEmissiveMask, nodeColor=<rgba 0 1 0.3672 1>}
  NamedRerouteDeclaration(None=@n1.EmissiveColor) {name=InEmissive, nodeColor=<rgba 0.4766 1 0 1>}
 ]}
