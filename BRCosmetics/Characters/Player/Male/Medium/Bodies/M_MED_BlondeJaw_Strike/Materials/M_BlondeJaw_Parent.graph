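# Parent material for the BlondeJaw Strike character outfit, layering a reactive glow effect (MF_BlondeJaw) onto skeletal mesh body and eye surfaces.
# Static switch IsEyes picks the shared eye function, with full glow mask and a scalar threshold as gradient, or the character PreFX function.
# The PreFX path takes glow, gradient and base emissive masks from channel-selectable ReactivityMasks texture.
# The chosen material attributes then pass through the shared character PostFX function.
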Material /BRCosmetics/Characters/Player/Male/Medium/Bodies/M_MED_BlondeJaw_Strike/Materials/M_BlondeJaw_Parent
graph {
 properties {
  bUseMaterialAttributes=true
  bUsedWithMeshDeformer=true
  bUsedWithMorphTargets=true
  bUsedWithSkeletalMesh=true
  pixelDepthOffsetMode=PDOM_Legacy
  shadingModel=MSM_FromMaterialExpression
  usedShadingModels=Default Lit | Subsurface | Clear Coat | Subsurface Profile | Cloth
 }
 nodes [
  n1 = MaterialFunctionCall() {materialFunction=/Game/Characters/Shared/FNEye/Materials/MF_eyes}
  n2 = Constant() {r=1}
  n3 = ScalarParameter() {defaultValue=0.5, group=BlondeJaw, parameterName=Eyes_GlowThreshold}
  n4 = MaterialFunctionCall() {materialFunction=&MF_CharacterParent_2023_PreFX}
  n5 = TextureSampleParameter2D() {
   group=BlondeJaw
   parameterName=ReactivityMasks
   samplerSource=SSM_Wrap_WorldGroupSettings
   samplerType=SAMPLERTYPE_Masks
   sortPriority=-50
   texture=&T_BlackMask
  }
  n6 = ChannelMaskParameter(None=@n5.RGBA) {group=BlondeJaw, parameterName=ReactivityMasks_GlowMaskChannel, sortPriority=-49}
  n7 = ChannelMaskParameter(None=@n5.RGBA) {group=BlondeJaw, maskChannel=Green, parameterName=ReactivityMasks_GradientChannel, sortPriority=-49, defaultValue=<rgba 0 1 0 0>}
  n8 = ChannelMaskParameter(None=@n5.RGBA) {group=BlondeJaw, maskChannel=Blue, parameterName=ReactivityMasks_BaseEmissiveMask, sortPriority=31, defaultValue=<rgba 0 0 1 0>}
  MaterialFunctionCall(
   Material=StaticSwitchParameter(
    True=MaterialFunctionCall!(In=@n1.Result, GlowMask=@n2, ReactivityGradient=@n3, BaseEmissiveMask) {materialFunction=&/MF_BlondeJaw}.SetMaterialAttributes
    False=MaterialFunctionCall(In=@n4.Result, GlowMask=@n6.R, ReactivityGradient=@n7.R, BaseEmissiveMask=@n8.R) {materialFunction=&MF_BlondeJaw}.SetMaterialAttributes
   ) {group=BlondeJaw, parameterName=IsEyes, sortPriority=-20}
   M=@n4.M
   Pre Skinned Local Position=@n4.M
   Distance Blend=@n4.M
   Fresnel=@n4.M
  ) {materialFunction=&MF_CharacterParent_2023_PostFX}
  ScalarParameter() {group=BlondeJaw, parameterName=DEBUG_Reactivity_MipLevel, sortPriority=-100}
 ]
}
