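# Corrective deformation Control Rig for the M_MED_BlondeJaw_Strike player body, driving helper bones from the main skeleton's pose.
# One execution chain constrains chest, back, inner-thigh and split-hand bones, then spreads swing-twist rotation along arm twist bones.
# Spherical pose readers on upper arms, elbows, knees and thighs weight transform offsets on corrective bones.
# Sides are not fully mirrored: the right-knee reader also moves left knee deform bones.
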ControlRigBlueprint /BRCosmetics/Characters/Player/Male/Medium/Bodies/M_MED_BlondeJaw_Strike/Meshes/M_MED_BlondeJaw_Strike_CtrlRig_DUP
graphs [
 RigVMFunctionLibrary {links [], nodes []}
 RigVMModel {nodes [
   RigUnit_BeginExecution = &RigUnit_BeginExecution()
   /Script/ControlRig.RigUnit_PositionConstraintLocalSpaceOffset(
    Child=(Type=Bone,Name="chest_l")
    Filter=(bX=True,bY=True,bZ=True)
    Parents=((Item=(Type=Bone,Name="upperarm_l"),Weight=0.5),(Item=(Type=Bone,Name="spine_04"),Weight=0.5))
    Weight=1
    bMaintainOffset=True
    ExecutePin=@RigUnit_BeginExecution.ExecutePin
   ) {is=struct RigVMUnitNode}
   > &RigUnit_PositionConstraintLocalSpaceOffset(
    Child=(Type=Bone,Name="chest_r")
    Filter=(bX=True,bY=True,bZ=True)
    Parents=((Item=(Type=Bone,Name="upperarm_r"),Weight=0.5),(Item=(Type=Bone,Name="spine_04"),Weight=0.5))
    Weight=1
    bMaintainOffset=True
   )
   > &RigUnit_PositionConstraintLocalSpaceOffset(
    Child=(Type=Bone,Name="back_l")
    Filter=(bX=True,bY=True,bZ=True)
    Parents=((Item=(Type=Bone,Name="upperarm_l"),Weight=0.5),(Item=(Type=Bone,Name="head"),Weight=0.5))
    Weight=1
    bMaintainOffset=True
   )
   > &RigUnit_PositionConstraintLocalSpaceOffset(
    Child=(Type=Bone,Name="back_r")
    Filter=(bX=True,bY=True,bZ=True)
    Parents=((Item=(Type=Bone,Name="upperarm_r"),Weight=0.5),(Item=(Type=Bone,Name="head"),Weight=0.5))
    Weight=1
    bMaintainOffset=True
   )
   > &RigUnit_PositionConstraintLocalSpaceOffset(
    Child=(Type=Bone,Name="deform_thigh_in_l")
    Filter=(bX=True,bY=True,bZ=True)
    Parents=((Item=(Type=Bone,Name="thigh_twist_01_l"),Weight=1))
    Weight=1
    bMaintainOffset=True
   )
   > &RigUnit_PositionConstraintLocalSpaceOffset(
    Child=(Type=Bone,Name="deform_thigh_in_r")
    Filter=(bX=True,bY=True,bZ=True)
    Parents=((Item=(Type=Bone,Name="thigh_twist_01_r"),Weight=1))
    Weight=1
    bMaintainOffset=True
   )
   RigVMFunction_MathQuaternionSwingTwist = &RigVMFunction_MathQuaternionSwingTwist(
    TwistAxis=(X=1,Y=0,Z=0)
    Input=&RigUnit_GetTransform(Item=(Type=Bone,Name="upperarm_l"), Space=LocalSpace, bInitial=False).Transform.Rotation
   )
   Inverse = /Script/RigVM.RigVMFunction_MathQuaternionInverse(Value=@RigVMFunction_MathQuaternionSwingTwist.Twist) {is=struct RigVMUnitNode, template="Inverse::Execute(in Value,out Result)"}
   > Set Transform = &RigUnit_SetRotation(Item=(Type=Bone,Name="deform_upperarm_twist_1_l"), Weight=0.8, bInitial=false, Value=@Inverse.Result)
   > Set Transform_1 = &RigUnit_SetRotation(Item=(Type=Bone,Name="deform_upperarm_twist_2_l"), Weight=0.6, bInitial=false, Value=@Inverse.Result)
   > Set Transform_1_1 = &RigUnit_SetRotation(Item=(Type=Bone,Name="deform_upperarm_twist_3_l"), Weight=0.4, bInitial=false, Value=@Inverse.Result)
   > Set Transform_1_1_1 = &RigUnit_SetRotation(Item=(Type=Bone,Name="deform_sdk_upperarm_twist_4_l"), Value=(X=0,Y=0,Z=0,W=1), Weight=0.1, bInitial=false)
   n1 = &RigUnit_GetTransform(Item=(Type=Bone,Name="upperarm_r"), Space=LocalSpace, bInitial=False)
   > Set Transform_2 = &RigUnit_SetRotation(
    Item=(Type=Bone,Name="deform_upperarm_twist_1_r")
    Weight=0.8
    bInitial=false
    Value=&RigVMFunction_MathQuaternionInverse(Value=&RigVMFunction_MathQuaternionSwingTwist(TwistAxis=(X=1,Y=0,Z=0), Input=@n1.Transform.Rotation).Twist) {template="Inverse::Execute(in Value,out Result)"}.Result
   )
   > Set Transform_1_2 = &RigUnit_SetRotation(Item=(Type=Bone,Name="deform_upperarm_twist_2_r"), Value=(X=0,Y=0,Z=0,W=1), Weight=0.6, bInitial=false)
   > Set Transform_1_1_2 = &RigUnit_SetRotation(Item=(Type=Bone,Name="deform_upperarm_twist_3_r"), Value=(X=0,Y=0,Z=0,W=1), Weight=0.4, bInitial=False)
   > Set Transform_1_1_2_1 = &RigUnit_SetRotation(Item=(Type=Bone,Name="deform_sdk_upperarm_twist_4_r"), Value=(X=0,Y=0,Z=0,W=1), Weight=0.1, bInitial=False)
   > &RigUnit_RotationConstraintLocalSpaceOffset(
    AdvancedSettings=(InterpolationType=Average,RotationOrderForFilter=XZY)
    Child=(Type=Bone,Name="hand_split_l")
    Filter=(bX=True,bY=True,bZ=True)
    Parents=((Item=(Type=Bone,Name="hand_l"),Weight=1))
    Weight=0.5
   )
   > &RigUnit_RotationConstraintLocalSpaceOffset(
    AdvancedSettings=(InterpolationType=Average,RotationOrderForFilter=XZY)
    Child=(Type=Bone,Name="hand_split_r")
    Filter=(bX=True,bY=True,bZ=True)
    Parents=((Item=(Type=Bone,Name="hand_r"),Weight=1))
    Weight=0.5
   )
   n2 = &RigVMFunction_MathQuaternionSwingTwist(
    TwistAxis=(X=1,Y=0,Z=0)
    Input=&RigUnit_GetTransform(Item=(Type=Bone,Name="hand_split_l"), Space=LocalSpace, bInitial=False).Transform.Rotation
   )
   > Set Transform_1_1_3 = &RigUnit_SetRotation(Item=(Type=Bone,Name="deform_lowerarm_twist_4_l"), Weight=0.8, bInitial=false, Value=@n2.Twist)
   > Set Transform_1_1_3_1 = &RigUnit_SetRotation(
    Item=(Type=Bone,Name="deform_lowerarm_twist_dbl_4_l")
    Weight=1
    bInitial=false
    Value=&RigUnit_GetTransform(Item=(Type=Bone,Name="deform_lowerarm_twist_4_l"), Space=LocalSpace, bInitial=False).Transform.Rotation
   )
   > &RigUnit_ModifyTransforms(
    ItemToModify=((Item=(Type=Bone,Name="deform_lowerarm_twist_4_l"),Transform=(Rotation=(X=-0.9511,Y=0,Z=0,W=0.309),Translation=(X=0,Y=0,Z=0),Scale3D=(X=1,Y=1,Z=1))))
    Weight=1
    WeightMinimum=0
   )
   > Set Transform_1_1_3_3 = &RigUnit_SetRotation(Item=(Type=Bone,Name="deform_lowerarm_twist_3_l"), Weight=0.6, bInitial=false, Value=@n2.Twist)
   > Set Transform_1_1_3_1_2 = &RigUnit_SetRotation(
    Item=(Type=Bone,Name="deform_lowerarm_twist_dbl_3_l")
    Weight=1
    bInitial=false
    Value=&RigUnit_GetTransform(Item=(Type=Bone,Name="deform_lowerarm_twist_3_l"), Space=LocalSpace, bInitial=False).Transform.Rotation
   )
   > &RigUnit_ModifyTransforms(
    ItemToModify=((Item=(Type=Bone,Name="deform_lowerarm_twist_3_l"),Transform=(Rotation=(X=-0.809,Y=0,Z=0,W=0.5878),Translation=(X=0,Y=0,Z=0),Scale3D=(X=1,Y=1,Z=1))))
    Weight=1
    WeightMinimum=0
   )
   > Set Transform_1_1_3_3_1 = &RigUnit_SetRotation(Item=(Type=Bone,Name="deform_lowerarm_twist_2_l"), Value=(X=0,Y=0,Z=0,W=1), Weight=0.3, bInitial=false)
   > Set Transform_1_1_3_1_2_1 = &RigUnit_SetRotation(
    Item=(Type=Bone,Name="deform_lowerarm_twist_dbl_2_l")
    Weight=1
    bInitial=false
    Value=&RigUnit_GetTransform(Item=(Type=Bone,Name="deform_lowerarm_twist_2_l"), Space=LocalSpace, bInitial=False).Transform.Rotation
   )
   > &RigUnit_ModifyTransforms(
    ItemToModify=((Item=(Type=Bone,Name="deform_lowerarm_twist_2_l"),Transform=(Rotation=(X=0,Y=0,Z=0,W=1),Translation=(X=0,Y=0,Z=0),Scale3D=(X=1,Y=1,Z=1))))
    Weight=0.3
    WeightMinimum=0
   )
   n3 = &RigVMFunction_MathQuaternionSwingTwist(
    TwistAxis=(X=1,Y=0,Z=0)
    Input=&RigUnit_GetTransform(Item=(Type=Bone,Name="hand_split_r"), Space=LocalSpace, bInitial=False).Transform.Rotation
   )
   > Set Transform_1_1_3_2 = &RigUnit_SetRotation(Item=(Type=Bone,Name="deform_lowerarm_twist_4_r"), Weight=0.8, bInitial=false, Value=@n3.Twist)
   > Set Transform_1_1_3_1_1 = &RigUnit_SetRotation(
    Item=(Type=Bone,Name="deform_lowerarm_twist_dbl_4_r")
    Weight=1
    bInitial=false
    Value=&RigUnit_GetTransform(Item=(Type=Bone,Name="deform_lowerarm_twist_4_r"), Space=LocalSpace, bInitial=False).Transform.Rotation
   )
   > &RigUnit_ModifyTransforms(
    ItemToModify=((Item=(Type=Bone,Name="deform_lowerarm_twist_4_r"),Transform=(Rotation=(X=-0.9511,Y=0,Z=0,W=0.309),Translation=(X=0,Y=0,Z=0),Scale3D=(X=1,Y=1,Z=1))))
    Weight=1
    WeightMinimum=0
   )
   > Set Transform_1_1_3_3_2 = &RigUnit_SetRotation(Item=(Type=Bone,Name="deform_lowerarm_twist_3_r"), Weight=0.6, bInitial=false, Value=@n3.Twist)
   > Set Transform_1_1_3_1_2_2 = &RigUnit_SetRotation(
    Item=(Type=Bone,Name="deform_lowerarm_twist_dbl_3_r")
    Weight=1
    bInitial=false
    Value=&RigUnit_GetTransform(Item=(Type=Bone,Name="deform_lowerarm_twist_3_r"), Space=LocalSpace, bInitial=False).Transform.Rotation
   )
   > &RigUnit_ModifyTransforms(
    ItemToModify=((Item=(Type=Bone,Name="deform_lowerarm_twist_3_r"),Transform=(Rotation=(X=-0.809,Y=0,Z=0,W=0.5878),Translation=(X=0,Y=0,Z=0),Scale3D=(X=1,Y=1,Z=1))))
    Weight=1
    WeightMinimum=0
   )
   > Set Transform_1_1_3_3_1_1 = &RigUnit_SetRotation(Item=(Type=Bone,Name="deform_lowerarm_twist_2_r"), Value=(X=0,Y=0,Z=0,W=1), Weight=0.3, bInitial=false)
   > Set Transform_1_1_3_1_2_1_1 = &RigUnit_SetRotation(
    Item=(Type=Bone,Name="deform_lowerarm_twist_dbl_2_r")
    Weight=1
    bInitial=false
    Value=&RigUnit_GetTransform(Item=(Type=Bone,Name="deform_lowerarm_twist_2_r"), Space=LocalSpace, bInitial=False).Transform.Rotation
   )
   > &RigUnit_ModifyTransforms(
    ItemToModify=((Item=(Type=Bone,Name="deform_lowerarm_twist_2_r"),Transform=(Rotation=(X=0,Y=0,Z=0,W=1),Translation=(X=0,Y=0,Z=0),Scale3D=(X=1,Y=1,Z=1))))
    Weight=0.3
    WeightMinimum=0
   )
   > n4 = &RigUnit_SphericalPoseReader(
    ActiveRegionScaleFactors=(PositiveWidth=0.1,NegativeWidth=0.1,PositiveHeight=0.1,NegativeHeight=0.1)
    ActiveRegionSize=0.1
    Debug=(bDrawDebug=false,bDraw2D=false,DebugScale=25,DebugSegments=20,DebugThickness=0.25)
    DriverAxis=(X=1,Y=0,Z=0)
    DriverItem=(Type=Bone,Name="lowerarm_l")
    FalloffRegionScaleFactors=(PositiveWidth=2,NegativeWidth=2,PositiveHeight=2,NegativeHeight=2)
    FalloffSize=0.45
    OptionalParentItem=(Type=Bone,Name="None")
    RotationOffset=(X=90,Y=0,Z=90)
   )
   > &RigUnit_ModifyTransforms(
    ItemToModify=((Item=(Type=Bone,Name="deform_elbow_in_l"),Transform=(Rotation=(X=0,Y=0,Z=0,W=1),Translation=(X=-20,Y=25,Z=0),Scale3D=(X=1,Y=1,Z=1))))
    WeightMinimum=0
    Weight=@n4.OutputParam
   )
   > &RigUnit_ModifyTransforms(
    ItemToModify=((Item=(Type=Bone,Name="deform_elbow_ot_l"),Transform=(Rotation=(X=0,Y=0,Z=0,W=1),Translation=(X=12,Y=-12,Z=0),Scale3D=(X=1,Y=1,Z=1))))
    WeightMinimum=0
    Weight=@n4.OutputParam
   )
   > n5 = &RigUnit_SphericalPoseReader(
    ActiveRegionScaleFactors=(PositiveWidth=0.1,NegativeWidth=0.1,PositiveHeight=0.1,NegativeHeight=0.1)
    ActiveRegionSize=0.1
    Debug=(bDrawDebug=false,bDraw2D=false,DebugScale=25,DebugSegments=20,DebugThickness=0.25)
    DriverAxis=(X=-1,Y=0,Z=0)
    DriverItem=(Type=Bone,Name="lowerarm_r")
    FalloffRegionScaleFactors=(PositiveWidth=2,NegativeWidth=2,PositiveHeight=2,NegativeHeight=2)
    FalloffSize=0.45
    OptionalParentItem=(Type=Bone,Name="None")
    RotationOffset=(X=0,Y=-90,Z=0)
   )
   > &RigUnit_ModifyTransforms(
    ItemToModify=((Item=(Type=Bone,Name="deform_elbow_in_r"),Transform=(Rotation=(X=0,Y=0,Z=0,W=1),Translation=(X=20,Y=-25,Z=0),Scale3D=(X=1,Y=1,Z=1))))
    WeightMinimum=0
    Weight=@n5.OutputParam
   )
   > &RigUnit_ModifyTransforms(
    ItemToModify=((Item=(Type=Bone,Name="deform_elbow_ot_r"),Transform=(Rotation=(X=0,Y=0,Z=0,W=1),Translation=(X=-12,Y=12,Z=0),Scale3D=(X=1,Y=1,Z=1))))
    WeightMinimum=0
    Weight=@n5.OutputParam
   )
   > n6 = &RigUnit_SphericalPoseReader(
    ActiveRegionScaleFactors=(PositiveWidth=0.1,NegativeWidth=0.1,PositiveHeight=0.1,NegativeHeight=0.1)
    ActiveRegionSize=0.1
    Debug=(bDrawDebug=true,bDraw2D=false,DebugScale=25,DebugSegments=20,DebugThickness=0.25)
    DriverAxis=(X=1,Y=0,Z=0)
    DriverItem=(Type=Bone,Name="upperarm_l")
    FalloffRegionScaleFactors=(PositiveWidth=1,NegativeWidth=1,PositiveHeight=1,NegativeHeight=1)
    FalloffSize=0.45
    OptionalParentItem=(Type=Bone,Name="spine_05")
    RotationOffset=(X=0,Y=0,Z=0)
   )
   > &RigUnit_ModifyTransforms(
    ItemToModify=((Item=(Type=Bone,Name="deform_sdk_upperarm_twist_1_l"),Transform=(Rotation=(X=0,Y=0,Z=0,W=1),Translation=(X=0,Y=-5,Z=5),Scale3D=(X=1,Y=1,Z=1))))
    WeightMinimum=0
    Weight=@n6.OutputParam
   )
   > &RigUnit_ModifyTransforms(
    ItemToModify=((Item=(Type=Bone,Name="lat_l"),Transform=(Rotation=(X=0,Y=0,Z=0,W=1),Translation=(X=22,Y=0,Z=0),Scale3D=(X=1,Y=1,Z=1))))
    WeightMinimum=0
    Weight=@n6.OutputParam
   )
   > n7 = &RigUnit_SphericalPoseReader(
    ActiveRegionScaleFactors=(PositiveWidth=0.1,NegativeWidth=0.1,PositiveHeight=0.1,NegativeHeight=0.1)
    ActiveRegionSize=0.1
    Debug=(bDrawDebug=false,bDraw2D=false,DebugScale=25,DebugSegments=20,DebugThickness=0.25)
    DriverAxis=(X=-1,Y=0,Z=0)
    DriverItem=(Type=Bone,Name="upperarm_r")
    FalloffRegionScaleFactors=(PositiveWidth=1,NegativeWidth=1,PositiveHeight=1,NegativeHeight=1)
    FalloffSize=0.45
    OptionalParentItem=(Type=Bone,Name="spine_05")
    RotationOffset=(X=180,Y=0,Z=0)
   )
   > &RigUnit_ModifyTransforms(
    ItemToModify=((Item=(Type=Bone,Name="deform_sdk_upperarm_twist_1_r"),Transform=(Rotation=(X=0,Y=0,Z=0,W=1),Translation=(X=0,Y=5,Z=-5),Scale3D=(X=1,Y=1,Z=1))))
    WeightMinimum=0
    Weight=@n7.OutputParam
   )
   > &RigUnit_ModifyTransforms(
    ItemToModify=((Item=(Type=Bone,Name="lat_r"),Transform=(Rotation=(X=0,Y=0,Z=0,W=1),Translation=(X=22,Y=0,Z=0),Scale3D=(X=1,Y=1,Z=1))))
    WeightMinimum=0
    Weight=@n7.OutputParam
   )
   > n8 = &RigUnit_SphericalPoseReader(
    ActiveRegionScaleFactors=(PositiveWidth=0.25,NegativeWidth=0.25,PositiveHeight=0.25,NegativeHeight=0.25)
    ActiveRegionSize=0.1
    Debug=(bDrawDebug=true,bDraw2D=false,DebugScale=25,DebugSegments=20,DebugThickness=0.25)
    DriverAxis=(X=-1,Y=0,Z=0)
    DriverItem=(Type=Bone,Name="upperarm_r")
    FalloffRegionScaleFactors=(PositiveWidth=1,NegativeWidth=1,PositiveHeight=1,NegativeHeight=1)
    FalloffSize=0.45
    OptionalParentItem=(Type=Bone,Name="spine_05")
    RotationOffset=(X=-100,Y=0,Z=45)
   )
   > &RigUnit_ModifyTransforms(
    ItemToModify=((Item=(Type=Bone,Name="deform_sdk_upperarm_twist_1_r"),Transform=(Rotation=(X=0.342,Y=0,Z=0,W=0.9397),Translation=(X=-8,Y=0,Z=-6),Scale3D=(X=1.2,Y=1.2,Z=1.2))))
    WeightMinimum=0
    Weight=@n8.OutputParam
   )
   > n9 = &RigUnit_SphericalPoseReader(
    ActiveRegionScaleFactors=(PositiveWidth=0.1,NegativeWidth=0.1,PositiveHeight=0.1,NegativeHeight=0.1)
    ActiveRegionSize=0.1
    Debug=(bDrawDebug=true,bDraw2D=false,DebugScale=25,DebugSegments=20,DebugThickness=0.25)
    DriverAxis=(X=-1,Y=0,Z=0)
    DriverItem=(Type=Bone,Name="calf_l")
    FalloffRegionScaleFactors=(PositiveWidth=0,NegativeWidth=1.5,PositiveHeight=1.5,NegativeHeight=1.5)
    FalloffSize=0.45
    OptionalParentItem=(Type=Bone,Name="thigh_l")
    RotationOffset=(X=-90,Y=0,Z=90)
   )
   > &RigUnit_ModifyTransforms(
    ItemToModify=((Item=(Type=Bone,Name="deform_knee_ot_l"),Transform=(Rotation=(X=0,Y=0,Z=0,W=1),Translation=(X=0,Y=5,Z=0),Scale3D=(X=1,Y=1,Z=1))))
    WeightMinimum=0
    Weight=@n9.OutputParam
   )
   > &RigUnit_ModifyTransforms(
    ItemToModify=((Item=(Type=Bone,Name="deform_knee_in_l"),Transform=(Rotation=(X=0,Y=0,Z=-0.5,W=0.866),Translation=(X=0,Y=-5,Z=0),Scale3D=(X=1,Y=1,Z=1))))
    WeightMinimum=0
    Weight=@n9.OutputParam
   )
   > &RigUnit_ModifyTransforms(
    ItemToModify=((Item=(Type=Bone,Name="calf_counter_l"),Transform=(Rotation=(X=0,Y=0,Z=-0.2588,W=0.9659),Translation=(X=0,Y=-2,Z=0),Scale3D=(X=1,Y=1,Z=1))))
    WeightMinimum=0
    Weight=@n9.OutputParam
   )
   > &RigUnit_ModifyTransforms(
    ItemToModify=((Item=(Type=Bone,Name="thigh_counter_l"),Transform=(Rotation=(X=0,Y=0,Z=0.342,W=0.9397),Translation=(X=0,Y=0,Z=0),Scale3D=(X=1,Y=1,Z=1))))
    WeightMinimum=0
    Weight=@n9.OutputParam
   )
   > n10 = &RigUnit_SphericalPoseReader(
    ActiveRegionScaleFactors=(PositiveWidth=0.1,NegativeWidth=0.1,PositiveHeight=0.1,NegativeHeight=0.1)
    ActiveRegionSize=0.1
    Debug=(bDrawDebug=false,bDraw2D=false,DebugScale=25,DebugSegments=20,DebugThickness=0.25)
    DriverAxis=(X=1,Y=0,Z=0)
    DriverItem=(Type=Bone,Name="calf_r")
    FalloffRegionScaleFactors=(PositiveWidth=0,NegativeWidth=1.5,PositiveHeight=1.5,NegativeHeight=1.5)
    FalloffSize=0.45
    OptionalParentItem=(Type=Bone,Name="thigh_r")
    RotationOffset=(X=-90,Y=0,Z=-90)
   )
   > &RigUnit_ModifyTransforms(
    ItemToModify=((Item=(Type=Bone,Name="deform_knee_ot_l"),Transform=(Rotation=(X=0,Y=0,Z=0,W=1),Translation=(X=0,Y=-5,Z=0),Scale3D=(X=1,Y=1,Z=1))))
    WeightMinimum=0
    Weight=@n10.OutputParam
   )
   > &RigUnit_ModifyTransforms(
    ItemToModify=((Item=(Type=Bone,Name="deform_knee_in_l"),Transform=(Rotation=(X=0,Y=0,Z=0.5,W=0.866),Translation=(X=0,Y=5,Z=0),Scale3D=(X=1,Y=1,Z=1))))
    WeightMinimum=0
    Weight=@n10.OutputParam
   )
   > &RigUnit_ModifyTransforms(
    ItemToModify=((Item=(Type=Bone,Name="calf_counter_r"),Transform=(Rotation=(X=0,Y=0,Z=-0.2588,W=0.9659),Translation=(X=0,Y=0,Z=0),Scale3D=(X=1,Y=1,Z=1))))
    WeightMinimum=0
    Weight=@n10.OutputParam
   )
   > &RigUnit_ModifyTransforms(
    ItemToModify=((Item=(Type=Bone,Name="thigh_counter_r"),Transform=(Rotation=(X=0,Y=0,Z=0.342,W=0.9397),Translation=(X=0,Y=0,Z=0),Scale3D=(X=1,Y=1,Z=1))))
    WeightMinimum=0
    Weight=@n10.OutputParam
   )
   > n11 = &RigUnit_SphericalPoseReader(
    ActiveRegionScaleFactors=(PositiveWidth=0.1,NegativeWidth=0.1,PositiveHeight=0.1,NegativeHeight=0.1)
    ActiveRegionSize=0.1
    Debug=(bDrawDebug=false,bDraw2D=false,DebugScale=25,DebugSegments=20,DebugThickness=0.25)
    DriverAxis=(X=-1,Y=0,Z=0)
    DriverItem=(Type=Bone,Name="thigh_l")
    FalloffRegionScaleFactors=(PositiveWidth=0,NegativeWidth=2,PositiveHeight=2,NegativeHeight=2)
    FalloffSize=0.45
    OptionalParentItem=(Type=Bone,Name="pelvis")
    RotationOffset=(X=90,Y=0,Z=-90)
   )
   > &RigUnit_ModifyTransforms(
    ItemToModify=((Item=(Type=Bone,Name="deform_thigh_in_l"),Transform=(Rotation=(X=0.0734,Y=-0.4162,Z=0.1574,W=0.8925),Translation=(X=15,Y=-20,Z=0),Scale3D=(X=1,Y=1,Z=1))))
    WeightMinimum=0
    Weight=@n11.OutputParam
   )
   > n12 = &RigUnit_SphericalPoseReader(
    ActiveRegionScaleFactors=(PositiveWidth=0.1,NegativeWidth=0.1,PositiveHeight=0.1,NegativeHeight=0.1)
    ActiveRegionSize=0.1
    Debug=(bDrawDebug=false,bDraw2D=false,DebugScale=25,DebugSegments=20,DebugThickness=0.25)
    DriverAxis=(X=1,Y=0,Z=0)
    DriverItem=(Type=Bone,Name="thigh_r")
    FalloffRegionScaleFactors=(PositiveWidth=0,NegativeWidth=2,PositiveHeight=2,NegativeHeight=2)
    FalloffSize=0.45
    OptionalParentItem=(Type=Bone,Name="pelvis")
    RotationOffset=(X=90,Y=0,Z=90)
   )
   > &RigUnit_ModifyTransforms(
    ItemToModify=((Item=(Type=Bone,Name="deform_thigh_in_r"),Transform=(Rotation=(X=0,Y=0,Z=0,W=1),Translation=(X=-15,Y=20,Z=0),Scale3D=(X=1,Y=1,Z=1))))
    WeightMinimum=0
    Weight=@n12.OutputParam
   )
   > n13 = &RigUnit_SphericalPoseReader(
    ActiveRegionScaleFactors=(PositiveWidth=0.1,NegativeWidth=0.1,PositiveHeight=0.1,NegativeHeight=0.1)
    ActiveRegionSize=0.1
    Debug=(bDrawDebug=true,bDraw2D=false,DebugScale=25,DebugSegments=20,DebugThickness=0.25)
    DriverAxis=(X=-1,Y=0,Z=0)
    DriverItem=(Type=Bone,Name="thigh_l")
    FalloffRegionScaleFactors=(PositiveWidth=0.5,NegativeWidth=0.5,PositiveHeight=0.5,NegativeHeight=0.5)
    FalloffSize=0.45
    OptionalParentItem=(Type=Bone,Name="pelvis")
    RotationOffset=(X=115,Y=0,Z=45)
   )
   > &RigUnit_ModifyTransforms(
    ItemToModify=((Item=(Type=Bone,Name="deform_thigh_in_l"),Transform=(Rotation=(X=0.0734,Y=-0.4162,Z=0.1574,W=0.8925),Translation=(X=5,Y=-20,Z=0),Scale3D=(X=1,Y=1,Z=1.75))))
    WeightMinimum=0
    Weight=@n13.OutputParam
   )
  ]}
]
hierarchy [
 root {loc=<0 0 0>, rot=<0 0 0 1>}
 pelvis {parent=root, loc=<0 -0.7071 93.1191>, rot=<0.0127 0.707 -0.0127 -0.707>}
 spine_01 {parent=pelvis, loc=<6.7494 -0.6388 0>, rot=<0 0 0.034 -0.9994>}
 spine_02 {parent=spine_01, loc=<7.2518 -0.035 0>, rot=<0 0 0.0238 -0.9997>}
 spine_03 {parent=spine_02, loc=<6.0104 0 0>, rot=<0 0 0.0149 -0.9999>}
 spine_04 {parent=spine_03, loc=<8.1024 0 0>, rot=<0 0 -0.0132 -0.9999>}
 spine_05 {parent=spine_04, loc=<10.0146 0 0>, rot=<0 0 -0.0015 -1>}
 attach_backpack {parent=spine_05, loc=<-0.123 -13.473 0>, rot=<-0.0093 -0.707 0.0093 -0.707>}
 neck_01 {parent=spine_05, loc=<11.3138 0.1529 0>, rot=<0 0 -0.2003 -0.9797>}
 neck_02 {parent=neck_01, loc=<4.5627 0 0>, rot=<0 0 0.0504 -0.9987>}
 head {parent=neck_02, loc=<5.3914 0 0>, rot=<0 0 0.115 -0.9934>}
 clavicle_l {parent=spine_05, loc=<4.3575 -0.1466 -2.322>, rot=<0.0465 -0.7289 -0.009 -0.6829>}
 upperarm_l {parent=clavicle_l, loc=<15.5111 0 0>, rot=<0.0226 -0.3234 -0.0088 -0.9459>}
 lowerarm_l {parent=upperarm_l, loc=<26.6514 0 0>, rot=<0.0391 -0.0099 -0.2453 -0.9686>}
 lowerarm_twist_02_l {parent=lowerarm_l, loc=<9.6199 0 0>, rot=<0 0 0 -1>}
 lowerarm_twist_01_l {parent=lowerarm_l, loc=<17.8549 0 0>, rot=<0 0 0 -1>}
 deform_lowerarm_twist_4_l {parent=lowerarm_l, loc=<22.3491 0 0>, rot=<0 0 0 1>}
 deform_lowerarm_twist_dbl_4_l {parent=deform_lowerarm_twist_4_l, loc=<0 0 0>, rot=<0 0 0 1>}
 deform_lowerarm_twist_3_l {parent=lowerarm_l, loc=<14.8994 0 0>, rot=<0 0 0 1>}
 deform_lowerarm_twist_dbl_3_l {parent=deform_lowerarm_twist_3_l, loc=<0 0 0>, rot=<0 0 0 1>}
 deform_lowerarm_twist_2_l {parent=lowerarm_l, loc=<7.4497 0 0>, rot=<0 0 0 1>}
 deform_lowerarm_twist_dbl_2_l {parent=deform_lowerarm_twist_2_l, loc=<0 0 0>, rot=<0 0 0 1>}
 deform_lowerarm_twist_1_l {parent=lowerarm_l, loc=<0 0 0>, rot=<0 0 0 1>}
 hand_noRot_l {parent=lowerarm_l, loc=<22.3491 0 0>, rot=<-0.7081 -0.0507 0.0975 -0.6975>}
 hand_split_l {parent=lowerarm_l, loc=<22.3491 0 0>, rot=<-0.7081 -0.0507 0.0975 -0.6975>}
 hand_l {parent=lowerarm_l, loc=<22.3494 -0.0001 0.0002>, rot=<-0.7081 -0.0507 0.0975 -0.6975>}
 thumb_01_l {parent=hand_l, loc=<1.3118 -0.3191 -2.459>, rot=<0.6546 -0.4042 0.0664 -0.6354>}
 thumb_02_l {parent=thumb_01_l, loc=<5.9336 -0.0448 0.0401>, rot=<-0.0037 0.0213 0.298 -0.9543>}
 thumb_03_l {parent=thumb_02_l, loc=<2.2223 0.3682 -0.0907>, rot=<0.0118 -0.013 0.0066 -0.9998>}
 index_metacarpal_l {parent=hand_l, loc=<1.6535 0.0183 -2.96>, rot=<0.0158 -0.0124 -0.0173 -0.9996>}
 index_01_l {parent=index_metacarpal_l, loc=<8.3008 0.0357 -0.0524>, rot=<-0.0044 0.0485 0.0924 -0.9945>}
 index_02_l {parent=index_01_l, loc=<3.3587 -0.3004 0.0452>, rot=<-0.002 0.0027 0.3225 -0.9466>}
 index_03_l {parent=index_02_l, loc=<2.5578 -0.123 -0.0252>, rot=<-0.002 0.0067 0.2558 -0.9667>}
 middle_metacarpal_l {parent=hand_l, loc=<2.1856 -0.1725 -0.5984>, rot=<0.006 0.0168 -0.0329 -0.9993>}
 middle_01_l {parent=middle_metacarpal_l, loc=<7.3275 -0.2447 0.0129>, rot=<-0.0208 0.0042 0.0858 -0.9961>}
 middle_02_l {parent=middle_01_l, loc=<4.0123 0.1401 0.1175>, rot=<-0.0218 0.0258 0.3512 -0.9357>}
 middle_03_l {parent=middle_02_l, loc=<2.6077 0.0534 0.1373>, rot=<0.0082 0.0188 0.1708 -0.9851>}
 ring_metacarpal_l {parent=hand_l, loc=<2.2074 -0.3625 1.3638>, rot=<-0.0284 0.0722 -0.0458 -0.9959>}
 ring_01_l {parent=ring_metacarpal_l, loc=<7.0295 -0.6197 -0.1446>, rot=<-0.0489 -0.0252 0.1892 -0.9804>}
 ring_02_l {parent=ring_01_l, loc=<3.6174 0.4647 -0.1027>, rot=<0 0 0.2806 -0.9598>}
 ring_03_l {parent=ring_02_l, loc=<2.4152 0.114 0.0199>, rot=<0 0 0.1189 -0.9929>}
 pinky_metacarpal_l {parent=hand_l, loc=<1.7616 -0.9753 3.3814>, rot=<-0.0397 0.0814 -0.069 -0.9935>}
 pinky_01_l {parent=pinky_metacarpal_l, loc=<7.1709 -0.4831 -0.1094>, rot=<-0.0654 -0.0035 0.2242 -0.9723>}
 pinky_02_l {parent=pinky_01_l, loc=<2.8274 0.0893 -0.1043>, rot=<0.0037 0.0011 0.2854 -0.9584>}
 pinky_03_l {parent=pinky_02_l, loc=<2.2046 0.0307 -0.0078>, rot=<-0.0002 0.0098 0.1445 -0.9895>}
 hand_attach_l {parent=hand_l, loc=<8.5 -3 0>, rot=<0.7071 0.7071 0 0>}
 weapon_l {parent=hand_l, loc=<-0.1974 -3.4111 0.3811>, rot=<0 0 0 1>}
 upperarm_twist_01_l {parent=upperarm_l, loc=<0.0041 0 0>, rot=<0 0 0 -1>}
 upperarm_twist_02_l {parent=upperarm_l, loc=<13.7941 0 0>, rot=<0 0 0 -1>}
 deform_upperarm_twist_4_l {parent=upperarm_l, loc=<26.6514 0 0>, rot=<0 0 0 1>}
 deform_sdk_upperarm_twist_4_l {parent=deform_upperarm_twist_4_l, loc=<0 0 0>, rot=<0 0 0 1>}
 deform_upperarm_twist_3_l {parent=upperarm_l, loc=<17.7689 0 0>, rot=<0 0 0 1>}
 deform_sdk_upperarm_twist_3_l {parent=deform_upperarm_twist_3_l, loc=<0 0 0>, rot=<0 0 0 1>}
 deform_upperarm_twist_2_l {parent=upperarm_l, loc=<8.8865 0 0>, rot=<0 0 0 1>}
 deform_sdk_upperarm_twist_2_l {parent=deform_upperarm_twist_2_l, loc=<0 0 0>, rot=<0 0 0 1>}
 deform_upperarm_twist_1_l {parent=upperarm_l, loc=<0 0 0>, rot=<0 0 0 1>}
 deform_sdk_upperarm_twist_1_l {parent=deform_upperarm_twist_1_l, loc=<0 0 0>, rot=<0 0 0 1>}
 deform_elbow_l {parent=upperarm_l, loc=<26.6514 0 0>, rot=<0.0197 -0.005 -0.1236 -0.9921>}
 deform_elbow_in_l {parent=deform_elbow_l, loc=<0 0 0>, rot=<0 0 0 1>}
 deform_elbow_ot_l {parent=deform_elbow_l, loc=<1.7577 -0.4708 1.6361>, rot=<0 0 0 1>}
 attach_cape {parent=spine_05, loc=<9.017 -11.714 0>, rot=<0.0616 -0.7044 -0.0616 -0.7044>}
 clavicle_r {parent=spine_05, loc=<4.3575 -0.1466 2.322>, rot=<0.7289 0.0465 -0.6829 0.009>}
 upperarm_r {parent=clavicle_r, loc=<-15.5111 0 0>, rot=<0.0226 -0.3234 -0.0088 -0.9459>}
 lowerarm_r {parent=upperarm_r, loc=<-26.6514 0 0>, rot=<0.0391 -0.0099 -0.2453 -0.9686>}
 hand_r {parent=lowerarm_r, loc=<-22.3491 0 0>, rot=<-0.7081 -0.0507 0.0975 -0.6975>}
 thumb_01_r {parent=hand_r, loc=<-1.3118 0.3191 2.459>, rot=<0.6546 -0.4042 0.0664 -0.6354>}
 thumb_02_r {parent=thumb_01_r, loc=<-5.9336 0.0448 -0.0401>, rot=<-0.0037 0.0213 0.298 -0.9543>}
 thumb_03_r {parent=thumb_02_r, loc=<-2.2223 -0.3682 0.0907>, rot=<0.0118 -0.013 0.0066 -0.9998>}
 index_metacarpal_r {parent=hand_r, loc=<-1.6535 -0.0183 2.96>, rot=<0.0158 -0.0124 -0.0173 -0.9996>}
 index_01_r {parent=index_metacarpal_r, loc=<-8.3008 -0.0357 0.0524>, rot=<-0.0044 0.0485 0.0924 -0.9945>}
 index_02_r {parent=index_01_r, loc=<-3.3587 0.3004 -0.0452>, rot=<-0.002 0.0027 0.3225 -0.9466>}
 index_03_r {parent=index_02_r, loc=<-2.5578 0.123 0.0252>, rot=<-0.002 0.0067 0.2558 -0.9667>}
 middle_metacarpal_r {parent=hand_r, loc=<-2.1856 0.1725 0.5984>, rot=<0.006 0.0168 -0.0329 -0.9993>}
 middle_01_r {parent=middle_metacarpal_r, loc=<-7.3275 0.2447 -0.0129>, rot=<-0.0208 0.0042 0.0858 -0.9961>}
 middle_02_r {parent=middle_01_r, loc=<-4.0123 -0.1401 -0.1175>, rot=<-0.0218 0.0258 0.3512 -0.9357>}
 middle_03_r {parent=middle_02_r, loc=<-2.6077 -0.0534 -0.1373>, rot=<0.0082 0.0188 0.1708 -0.9851>}
 ring_metacarpal_r {parent=hand_r, loc=<-2.2074 0.3625 -1.3638>, rot=<-0.0284 0.0722 -0.0458 -0.9959>}
 ring_01_r {parent=ring_metacarpal_r, loc=<-7.0295 0.6197 0.1446>, rot=<-0.0489 -0.0252 0.1892 -0.9804>}
 ring_02_r {parent=ring_01_r, loc=<-3.6174 -0.4647 0.1027>, rot=<0 0 0.2806 -0.9598>}
 ring_03_r {parent=ring_02_r, loc=<-2.4152 -0.114 -0.0199>, rot=<0 0 0.1189 -0.9929>}
 pinky_metacarpal_r {parent=hand_r, loc=<-1.7616 0.9753 -3.3814>, rot=<-0.0397 0.0814 -0.069 -0.9935>}
 pinky_01_r {parent=pinky_metacarpal_r, loc=<-7.1709 0.4831 0.1094>, rot=<-0.0654 -0.0035 0.2242 -0.9723>}
 pinky_02_r {parent=pinky_01_r, loc=<-2.8274 -0.0893 0.1043>, rot=<0.0037 0.0011 0.2854 -0.9584>}
 pinky_03_r {parent=pinky_02_r, loc=<-2.2046 -0.0307 0.0078>, rot=<-0.0002 0.0098 0.1445 -0.9895>}
 weapon_r {parent=hand_r, loc=<0.1974 3.4111 -0.3811>, rot=<0 0 0 1>}
 hand_attach_r {parent=hand_r, loc=<-8.5001 2.9999 0>, rot=<0 0 -0.7071 -0.7071>}
 lowerarm_twist_02_r {parent=lowerarm_r, loc=<-9.6199 0 0>, rot=<0 0 0 -1>}
 lowerarm_twist_01_r {parent=lowerarm_r, loc=<-17.8549 0 0>, rot=<0 0 0 -1>}
 deform_lowerarm_twist_2_r {parent=lowerarm_r, loc=<-7.4499 0.0001 0.0002>, rot=<0 0 0 1>}
 deform_lowerarm_twist_dbl_2_r {parent=deform_lowerarm_twist_2_r, loc=<0 0 0>, rot=<0 0 0 -1>}
 deform_lowerarm_twist_1_r {parent=lowerarm_r, loc=<-0.0001 0 0.0002>, rot=<0 0 0 -1>}
 deform_lowerarm_twist_3_r {parent=lowerarm_r, loc=<-14.8997 0.0001 0.0003>, rot=<0 0 0 -1>}
 deform_lowerarm_twist_dbl_3_r {parent=deform_lowerarm_twist_3_r, loc=<0 0 0>, rot=<0 0 0 -1>}
 deform_lowerarm_twist_4_r {parent=lowerarm_r, loc=<-22.3488 -0.0001 -0.0003>, rot=<0 0 0 1>}
 deform_lowerarm_twist_dbl_4_r {parent=deform_lowerarm_twist_4_r, loc=<0 0 0>, rot=<0 0 0 -1>}
 hand_noRot_r {parent=lowerarm_r, loc=<-22.3491 0 0>, rot=<-0.7081 -0.0507 0.0975 -0.6975>}
 hand_split_r {parent=lowerarm_r, loc=<-22.3491 0 0>, rot=<-0.7081 -0.0507 0.0975 -0.6975>}
 upperarm_twist_01_r {parent=upperarm_r, loc=<-0.0041 0 0>, rot=<0 0 0 -1>}
 upperarm_twist_02_r {parent=upperarm_r, loc=<-13.7941 0 0>, rot=<0 0 0 -1>}
 deform_upperarm_twist_4_r {parent=upperarm_r, loc=<-26.6515 0 0.0002>, rot=<0 0 0 1>}
 deform_sdk_upperarm_twist_4_r {parent=deform_upperarm_twist_4_r, loc=<0 0 0>, rot=<0 0 0 1>}
 deform_upperarm_twist_3_r {parent=upperarm_r, loc=<-17.7689 0 0>, rot=<0 0 0 1>}
 deform_sdk_upperarm_twist_3_r {parent=deform_upperarm_twist_3_r, loc=<0 0 0>, rot=<0 0 0 1>}
 deform_upperarm_twist_2_r {parent=upperarm_r, loc=<-8.8863 0 -0.0002>, rot=<0 0 0 1>}
 deform_sdk_upperarm_twist_2_r {parent=deform_upperarm_twist_2_r, loc=<0 0 0>, rot=<0 0 0 1>}
 deform_upperarm_twist_1_r {parent=upperarm_r, loc=<-0.0001 0 0.0001>, rot=<0 0 0 1>}
 deform_sdk_upperarm_twist_1_r {parent=deform_upperarm_twist_1_r, loc=<0 0 0>, rot=<0 0 0 1>}
 deform_elbow_r {parent=upperarm_r, loc=<-26.6515 0 0.0002>, rot=<0.0197 -0.005 -0.1236 -0.9921>}
 deform_elbow_in_r {parent=deform_elbow_r, loc=<0 0 0>, rot=<0 0 0 1>}
 deform_elbow_ot_r {parent=deform_elbow_r, loc=<0 0 0>, rot=<0 0 0 1>}
 spike_01_l {parent=spine_05, loc=<14.698 -14.2103 -5.9485>, rot=<0 0 0 1>}
 spike_01_r {parent=spine_05, loc=<14.698 -14.2103 5.949>, rot=<0 0 0 1>}
 spike_01 {parent=spine_05, loc=<12.6882 -14.3719 0>, rot=<0 0 0 1>}
 back_r {parent=spine_04, loc=<6.7002 -0.4985 8.8893>, rot=<0.7065 0.0294 -0.7065 0.0294>}
 back_l {parent=spine_04, loc=<6.7002 -0.4985 -8.8893>, rot=<0.0294 -0.7065 -0.0294 -0.7065>}
 spike_02_r {parent=spine_04, loc=<13.0874 -19.7005 7.167>, rot=<0 0 -0.0015 -1>}
 spike_02 {parent=spine_04, loc=<8.1102 -20.933 0>, rot=<0 0 -0.0015 -1>}
 spike_02_l {parent=spine_04, loc=<13.0874 -19.7005 -7.1674>, rot=<0 0 -0.0015 -1>}
 chest_r {parent=spine_04, loc=<6.7002 -0.4985 8.8893>, rot=<0.7065 0.0294 -0.7065 0.0294>}
 chest_l {parent=spine_04, loc=<6.7002 -0.4985 -8.8893>, rot=<0.0294 -0.7065 -0.0294 -0.7065>}
 spike_03_r {parent=spine_03, loc=<5.0963 -21.1755 6.426>, rot=<0 0 -0.0147 -0.9999>}
 spike_03_l {parent=spine_03, loc=<5.0963 -21.1755 -6.4258>, rot=<0 0 -0.0147 -0.9999>}
 spike_03 {parent=spine_03, loc=<0.1532 -22.5386 0>, rot=<0 0 -0.0147 -0.9999>}
 lat_l {parent=spine_02, loc=<6.0104 0 -15>, rot=<0 0 0.0149 -0.9999>}
 lat_r {parent=spine_02, loc=<6.0104 0 15>, rot=<0 0 0.0149 -0.9999>}
 thigh_l {parent=pelvis, loc=<-2.5693 -0.0109 -7.7501>, rot=<0.0012 0.0572 0.0155 -0.9982>}
 calf_l {parent=thigh_l, loc=<-39.0484 0 0>, rot=<0.0006 -0.0161 -0.0391 -0.9991>}
 foot_l {parent=calf_l, loc=<-42.561 0.1077 0.4292>, rot=<0.0418 -0.0455 0.0343 -0.9975>}
 ball_l {parent=foot_l, loc=<-6.1599 13.8264 0.6806>, rot=<0.0001 -0.0001 -0.7063 -0.7079>}
 calf_twist_02_l {parent=calf_l, loc=<-18.8064 -0.0967 0.0397>, rot=<0.0237 0.0095 -0.0035 -0.9997>}
 calf_twist_01_l {parent=calf_l, loc=<-35.813 -0.0967 0.0397>, rot=<0.0237 0.0095 -0.0035 -0.9997>}
 deform_knee_l {parent=calf_l, loc=<0 0 0>, rot=<0 0 0 1>}
 deform_knee_in_l {parent=deform_knee_l, loc=<0 0 0>, rot=<0 0 0 1>}
 deform_knee_ot_l {parent=deform_knee_l, loc=<0 0 0>, rot=<0 0 0 1>}
 calf_counter_l {parent=calf_l, loc=<0 0 0>, rot=<0 0 0 1>}
 thigh_twist_01_l {parent=thigh_l, loc=<-21.6303 0 0>, rot=<0.0013 0 0 -1>}
 thigh_counter_l {parent=thigh_l, loc=<-39.0484 0 0>, rot=<0.0006 -0.0161 -0.0391 -0.9991>}
 thigh_r {parent=pelvis, loc=<-2.5693 -0.0109 7.7501>, rot=<0.0572 -0.0012 0.9982 0.0155>}
 calf_r {parent=thigh_r, loc=<39.0484 0 0>, rot=<0.0006 -0.0161 -0.0391 -0.9991>}
 foot_r {parent=calf_r, loc=<42.561 -0.1077 -0.4292>, rot=<0.0418 -0.0455 0.0343 -0.9975>}
 ball_r {parent=foot_r, loc=<6.1599 -13.8264 -0.6806>, rot=<0.0001 -0.0001 -0.7063 -0.7079>}
 calf_twist_02_r {parent=calf_r, loc=<18.8064 0.0967 -0.0397>, rot=<0.0237 0.0095 -0.0035 -0.9997>}
 calf_twist_01_r {parent=calf_r, loc=<35.813 0.0967 -0.0397>, rot=<0.0237 0.0095 -0.0035 -0.9997>}
 deform_knee_r {parent=calf_r, loc=<0 0 0>, rot=<0 0 0 -1>}
 deform_knee_in_r {parent=deform_knee_r, loc=<0 0 0>, rot=<0 0 0 -1>}
 deform_knee_ot_r {parent=deform_knee_r, loc=<0 0 0>, rot=<0 0 0 -1>}
 calf_counter_r {parent=calf_r, loc=<0 0 0>, rot=<0 0 0 1>}
 thigh_twist_01_r {parent=thigh_r, loc=<21.6303 0 0>, rot=<0.0013 0 0 -1>}
 thigh_counter_r {parent=thigh_r, loc=<39.0484 0 0>, rot=<0.0006 -0.0161 -0.0391 -0.9991>}
 deform_thigh_in_l {parent=pelvis, loc=<-24.0479 0.6538 -10.2194>, rot=<0.0025 0.0571 0.0155 -0.9982>}
 deform_thigh_in_r {parent=pelvis, loc=<-24.0479 0.6538 10.2194>, rot=<0.0571 -0.0025 0.9982 0.0155>}
 tail1 {parent=pelvis, loc=<-16.3746 -8.6704 0>, rot=<0.282 -0.6485 0.6485 -0.282>}
 tail2 {parent=tail1, loc=<24.6384 0 0>, rot=<0 0.0793 0 -0.9968>}
 tail3 {parent=tail2, loc=<26.1413 0 0>, rot=<0 0.0996 0 -0.995>}
 tail4 {parent=tail3, loc=<25.2556 0 0>, rot=<0 0 0 1>}
 deform_pelvis {parent=pelvis, loc=<0 0 0>, rot=<0 0 0 1>}
 attach {parent=root, loc=<0 0 0>, rot=<0 0 0 1>}
 ik_hand_root {parent=root, loc=<0 0 0>, rot=<0 0 0 1>}
 ik_hand_gun {parent=ik_hand_root, loc=<-51.9727 4.0883 102.6935>, rot=<0.6084 0.1793 0.4613 -0.6205>}
 ik_hand_r {parent=ik_hand_gun, loc=<0 0 0>, rot=<0 0 0 1>}
 ik_hand_l {parent=ik_hand_gun, loc=<53.0285 82.1781 35.2051>, rot=<-0.5102 -0.7906 -0.3387 0>}
 ik_foot_root {parent=root, loc=<0 0 0>, rot=<0 0 0 1>}
 ik_foot_l {parent=ik_foot_root, loc=<15.2608 -4.4176 9.4432>, rot=<0.0357 0.7045 0.0284 -0.7082>}
 ik_foot_r {parent=ik_foot_root, loc=<-15.2608 -4.4176 9.4432>, rot=<0.7082 0.0284 -0.7045 0.0357>}
 root_fp {parent=root, loc=<0 0 0>, rot=<0 0 0 1>}
]
