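# Post-process style AnimBlueprint for the BlondeJaw Strike character body, copying the parent mesh pose through a Control Rig and animating its tail and spikes.
# An emote flag selects between two tail AnimDynamics branches built on the cached Presim pose.
# Additive bone offsets then adjust for hang gliding, sprinting, skydiving, swimming, sliding, back bling and the frontend backpack pose.
# A final layered blend lets DefaultSlot override the tail during custom idles.
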AnimBlueprint /BRCosmetics/Characters/Player/Male/Medium/Bodies/M_MED_BlondeJaw_Strike/Meshes/M_MED_BlondeJaw_Strike_AnimBP
target_skeleton=&/M_MED_BlondeJaw_Strike_Skeleton
graphs [
 AnimGraph {nodes [
   n1 = Get Is Baseball Sliding float() -> (Is Baseball Sliding float:double)
   "Constraint - tail1 (R)"(Alpha=@n1, ConstraintWeights_0=0) {
    alphaBoolBlend {blendOption=Linear}
    boneToModify {boneName=tail1}
    constraintSetup [{transformType=Rotation, targetBone {boneName=root}}]
    constraintWeights [1]
   }
   n2 = Get Is Tactical Sprinting() -> (Is Tactical Sprinting:double)
   n3 = "Transform (Modify) Bone - Bone: tail4"(Alpha=@n2, Translation="0,0,0", Rotation="0,0,-20", Scale="1,1,1") {
    rotationMode=BMM_Additive
    scaleMode=BMM_Additive
    translationMode=BMM_Additive
    alphaBoolBlend {blendOption=Linear}
    alphaScaleBiasClamp {bInterpResult=true, interpSpeedDecreasing=1, interpSpeedIncreasing=1}
    boneToModify {boneName=tail4}
    rotation {roll=-20}
   }
   "Transform (Modify) Bone - Bone: tail3"(ComponentPose=@n3, Alpha=@n2, Translation="0,0,0", Rotation="0,0,-20", Scale="1,1,1") {
    rotationMode=BMM_Additive
    scaleMode=BMM_Additive
    translationMode=BMM_Additive
    alphaBoolBlend {blendOption=Linear}
    alphaScaleBiasClamp {bInterpResult=true, interpSpeedDecreasing=1, interpSpeedIncreasing=1}
    boneToModify {boneName=tail3}
    rotation {roll=-20}
   }
   n4 = Get animDynAlpha() -> (animDynAlpha:double)
   n5 = Anim Dynamics - Bone: tail1(ComponentPose=Local To Component(LocalPose=Use cached pose 'Presim'()), Alpha=@n4, ExternalForce="0,0,0") {
    angularDampingOverride=1
    angularSpringConstant=5000
    bAngularSpring=true
    bOverrideAngularDamping=true
    bUseGravityOverride=true
    gravityScale=0.35
    simulationSpace=RootRelative
    alphaBoolBlend {blendOption=Linear}
    boundBone {boneName=tail1}
    componentLinearAccScale=<1 1 1>
    componentLinearVelScale=<1 1 1>
    gravityOverride {z=-1500}
    physicsBodyDefinitions [{
      boundBone {boneName=tail1}
      boxExtents=<23 23 23>
      constraintSetup {angularLimitsMax=<0 30 30>, angularLimitsMin=<-30 -30 -30>, angularTarget=<1 0 0>}
      localJointOffset=<1 0 0>
     }]
   }
   n6 = Anim Dynamics - Bone: tail2(ComponentPose=@n5, Alpha=@n4, ExternalForce="0,0,0") {
    angularDampingOverride=1
    angularSpringConstant=5000
    bAngularSpring=true
    bOverrideAngularDamping=true
    bUseGravityOverride=true
    gravityScale=0.35
    simulationSpace=RootRelative
    alphaBoolBlend {blendOption=Linear}
    boundBone {boneName=tail2}
    componentLinearAccScale=<1 1 1>
    componentLinearVelScale=<1 1 1>
    gravityOverride {z=-1500}
    physicsBodyDefinitions [{
      boundBone {boneName=tail2}
      boxExtents=<23 23 23>
      constraintSetup {angularLimitsMax=<0 30 30>, angularLimitsMin=<-30 -30 -30>, angularTarget=<1 0 0>}
      localJointOffset=<1 0 0>
     }]
   }
   n7 = Anim Dynamics - Bone: tail3(ComponentPose=@n6, Alpha=@n4, ExternalForce="0,0,0") {
    angularDampingOverride=1
    angularSpringConstant=5000
    bAngularSpring=true
    bOverrideAngularDamping=true
    bUseGravityOverride=true
    gravityScale=0.35
    simulationSpace=RootRelative
    alphaBoolBlend {blendOption=Linear}
    boundBone {boneName=tail3}
    componentLinearAccScale=<1 1 1>
    componentLinearVelScale=<1 1 1>
    gravityOverride {z=-1500}
    physicsBodyDefinitions [{
      boundBone {boneName=tail3}
      boxExtents=<23 23 23>
      constraintSetup {angularLimitsMax=<0 30 30>, angularLimitsMin=<-30 -30 -30>, angularTarget=<1 0 0>}
      localJointOffset=<1 0 0>
     }]
   }
   n8 = Anim Dynamics - Bone: tail4(ComponentPose=@n7, Alpha=@n4, ExternalForce="0,0,0") {
    angularDampingOverride=1
    angularSpringConstant=5000
    bAngularSpring=true
    bOverrideAngularDamping=true
    bUseGravityOverride=true
    gravityScale=0.35
    simulationSpace=RootRelative
    alphaBoolBlend {blendOption=Linear}
    boundBone {boneName=tail4}
    componentLinearAccScale=<1 1 1>
    componentLinearVelScale=<1 1 1>
    gravityOverride {z=-1500}
    physicsBodyDefinitions [{
      boundBone {boneName=tail4}
      boxExtents=<23 23 23>
      constraintSetup {angularLimitsMax=<0 30 30>, angularLimitsMin=<-30 -30 -30>, angularTarget=<1 0 0>}
      localJointOffset=<1 0 0>
     }]
   }
   n9 = Get animDynAlpha()
   n10 = "Transform (Modify) Bone - Bone: tail1"(ComponentPose=@n8, Alpha=@n9, Translation="0,0,0", Rotation="0,0,20", Scale="1,1,1") {
    rotationMode=BMM_Additive
    scaleMode=BMM_Additive
    translationMode=BMM_Additive
    alphaBoolBlend {blendOption=Linear}
    boneToModify {boneName=tail1}
    rotation {roll=20}
   }
   n11 = "Transform (Modify) Bone - Bone: tail2"(ComponentPose=@n10, Alpha=@n9, Translation="0,0,0", Rotation="0,0,20", Scale="1,1,1") {
    rotationMode=BMM_Additive
    scaleMode=BMM_Additive
    translationMode=BMM_Additive
    alphaBoolBlend {blendOption=Linear}
    boneToModify {boneName=tail2}
    rotation {roll=20}
   }
   n12 = "Transform (Modify) Bone - Bone: tail3"(ComponentPose=@n11, Alpha=@n9, Translation="0,0,0", Rotation="0,0,20", Scale="1,1,1") {
    rotationMode=BMM_Additive
    scaleMode=BMM_Additive
    translationMode=BMM_Additive
    alphaBoolBlend {blendOption=Linear}
    boneToModify {boneName=tail3}
    rotation {roll=20}
   }
   n13 = "Transform (Modify) Bone - Bone: tail4"(ComponentPose=@n12, Alpha=@n9, Translation="0,0,0", Rotation="0,0,20", Scale="1,1,1") {
    rotationMode=BMM_Additive
    scaleMode=BMM_Additive
    translationMode=BMM_Additive
    alphaBoolBlend {blendOption=Linear}
    boneToModify {boneName=tail4}
    rotation {roll=20}
   }
   n14 = Component To Local(ComponentPose=@n13)
   n15 = Get In Vehicle Sim Alpha() -> (In Vehicle Sim Alpha:double)
   n16 = Anim Dynamics - Bone: tail1(ComponentPose=Local To Component(LocalPose=Use cached pose 'Presim'()), Alpha=@n15, ExternalForce="0,0,0") {
    angularDampingOverride=1
    angularSpringConstant=5000
    bAngularSpring=true
    bOverrideAngularDamping=true
    bUseGravityOverride=true
    gravityScale=0.35
    simulationSpace=RootRelative
    alphaBoolBlend {blendOption=Linear}
    boundBone {boneName=tail1}
    componentLinearAccScale=<1 1 1>
    componentLinearVelScale=<1 1 1>
    gravityOverride {z=-1500}
    physicsBodyDefinitions [{
      boundBone {boneName=tail1}
      boxExtents=<23 23 23>
      constraintSetup {angularLimitsMax=<0 30 30>, angularLimitsMin=<-30 -30 -30>, angularTarget=<1 0 0>}
      localJointOffset=<1 0 0>
     }]
   }
   n17 = Anim Dynamics - Bone: tail2(ComponentPose=@n16, Alpha=@n15, ExternalForce="0,0,0") {
    angularDampingOverride=1
    angularSpringConstant=5000
    bAngularSpring=true
    bOverrideAngularDamping=true
    bUseGravityOverride=true
    gravityScale=0.35
    simulationSpace=RootRelative
    alphaBoolBlend {blendOption=Linear}
    boundBone {boneName=tail2}
    componentLinearAccScale=<1 1 1>
    componentLinearVelScale=<1 1 1>
    gravityOverride {z=-1500}
    physicsBodyDefinitions [{
      boundBone {boneName=tail2}
      boxExtents=<23 23 23>
      constraintSetup {angularLimitsMax=<0 30 30>, angularLimitsMin=<-30 -30 -30>, angularTarget=<1 0 0>}
      localJointOffset=<1 0 0>
     }]
   }
   n18 = Anim Dynamics - Bone: tail3(ComponentPose=@n17, Alpha=@n15, ExternalForce="0,0,0") {
    angularDampingOverride=1
    angularSpringConstant=5000
    bAngularSpring=true
    bOverrideAngularDamping=true
    bUseGravityOverride=true
    gravityScale=0.35
    simulationSpace=RootRelative
    alphaBoolBlend {blendOption=Linear}
    boundBone {boneName=tail3}
    componentLinearAccScale=<1 1 1>
    componentLinearVelScale=<1 1 1>
    gravityOverride {z=-1500}
    physicsBodyDefinitions [{
      boundBone {boneName=tail3}
      boxExtents=<23 23 23>
      constraintSetup {angularLimitsMax=<0 30 30>, angularLimitsMin=<-30 -30 -30>, angularTarget=<1 0 0>}
      localJointOffset=<1 0 0>
     }]
   }
   n19 = Anim Dynamics - Bone: tail4(ComponentPose=@n18, Alpha=@n15, ExternalForce="0,0,0") {
    angularDampingOverride=1
    angularSpringConstant=5000
    bAngularSpring=true
    bOverrideAngularDamping=true
    bUseGravityOverride=true
    gravityScale=0.35
    simulationSpace=RootRelative
    alphaBoolBlend {blendOption=Linear}
    boundBone {boneName=tail4}
    componentLinearAccScale=<1 1 1>
    componentLinearVelScale=<1 1 1>
    gravityOverride {z=-1500}
    physicsBodyDefinitions [{
      boundBone {boneName=tail4}
      boxExtents=<23 23 23>
      constraintSetup {angularLimitsMax=<0 30 30>, angularLimitsMin=<-30 -30 -30>, angularTarget=<1 0 0>}
      localJointOffset=<1 0 0>
     }]
   }
   n20 = Get in vehicle scale alpha() -> (in vehicle scale alpha:double)
   n21 = "Transform (Modify) Bone - Bone: tail1"(ComponentPose=@n19, Alpha=@n20, Translation="0,0,0", Rotation="0,0,20", Scale="0.1,0.1,0.1") {
    rotationMode=BMM_Additive
    scaleMode=BMM_Additive
    translationMode=BMM_Additive
    alphaBoolBlend {blendOption=Linear}
    boneToModify {boneName=tail1}
    rotation {roll=20}
    scale=<0.1 0.1 0.1>
   }
   n22 = "Transform (Modify) Bone - Bone: tail2"(ComponentPose=@n21, Alpha=@n20, Translation="0,0,0", Rotation="0,0,20", Scale="0.1,0.1,0.1") {
    rotationMode=BMM_Additive
    scaleMode=BMM_Additive
    translationMode=BMM_Additive
    alphaBoolBlend {blendOption=Linear}
    boneToModify {boneName=tail2}
    rotation {roll=20}
    scale=<0.1 0.1 0.1>
   }
   n23 = "Transform (Modify) Bone - Bone: tail3"(ComponentPose=@n22, Alpha=@n20, Translation="0,0,0", Rotation="0,0,20", Scale="0.1,0.1,0.1") {
    rotationMode=BMM_Additive
    scaleMode=BMM_Additive
    translationMode=BMM_Additive
    alphaBoolBlend {blendOption=Linear}
    boneToModify {boneName=tail3}
    rotation {roll=20}
    scale=<0.1 0.1 0.1>
   }
   n24 = "Transform (Modify) Bone - Bone: tail4"(ComponentPose=@n23, Alpha=@n20, Translation="0,0,0", Rotation="0,0,20", Scale="0.1,0.1,0.1") {
    rotationMode=BMM_Additive
    scaleMode=BMM_Additive
    translationMode=BMM_Additive
    alphaBoolBlend {blendOption=Linear}
    boneToModify {boneName=tail4}
    rotation {roll=20}
    scale=<0.1 0.1 0.1>
   }
   n25 = Component To Local(ComponentPose=@n24)
   n26 = Get bIsPlayingEmote()
   n27 = Get IsHangGlider() -> (IsHangGlider:bool)
   n28 = Anim Dynamics - Bone: tail2!(
    ComponentPose=Local To Component(LocalPose=Blend Poses by bool(BlendPose_0=@n25, BlendPose_1=@n14, bActiveValue=@n26) {blendType=Linear})
    bAlphaBoolEnabled:bool=@n27
    ExternalForce="0,0,0"
   ) -> (Pose) {
    alphaInputType=Bool
    angularDampingOverride=1
    angularSpringConstant=5000
    bAngularSpring=true
    bOverrideAngularDamping=true
    bUseGravityOverride=true
    gravityScale=0.35
    simulationSpace=RootRelative
    alphaBoolBlend {blendOption=Linear}
    boundBone {boneName=tail2}
    componentLinearAccScale=<1 1 1>
    componentLinearVelScale=<1 1 1>
    gravityOverride {z=-1500}
    physicsBodyDefinitions [{
      boundBone {boneName=tail2}
      boxExtents=<30 30 30>
      constraintSetup {angularLimitsMax=<0 30 30>, angularLimitsMin=<-30 -30 -30>, angularTarget=<1 0 0>}
      localJointOffset=<1 0 0>
     }]
   }
   n29 = Anim Dynamics - Bone: tail3!(ComponentPose=@n28, bAlphaBoolEnabled:bool=@n27, ExternalForce="0,0,0") -> (Pose) {
    alphaInputType=Bool
    angularDampingOverride=1
    angularSpringConstant=5000
    bAngularSpring=true
    bOverrideAngularDamping=true
    bUseGravityOverride=true
    gravityScale=0.35
    simulationSpace=RootRelative
    alphaBoolBlend {blendOption=Linear}
    boundBone {boneName=tail3}
    componentLinearAccScale=<1 1 1>
    componentLinearVelScale=<1 1 1>
    gravityOverride {z=-1500}
    physicsBodyDefinitions [{
      boundBone {boneName=tail3}
      boxExtents=<30 30 30>
      constraintSetup {angularLimitsMax=<0 30 30>, angularLimitsMin=<-30 -30 -30>, angularTarget=<1 0 0>}
      localJointOffset=<1 0 0>
     }]
   }
   n30 = Anim Dynamics - Bone: tail4!(ComponentPose=@n29, bAlphaBoolEnabled:bool=@n27, ExternalForce="0,0,0") -> (Pose) {
    alphaInputType=Bool
    angularDampingOverride=1
    angularSpringConstant=5000
    bAngularSpring=true
    bOverrideAngularDamping=true
    bUseGravityOverride=true
    gravityScale=0.35
    simulationSpace=RootRelative
    alphaBoolBlend {blendOption=Linear}
    boundBone {boneName=tail4}
    componentLinearAccScale=<1 1 1>
    componentLinearVelScale=<1 1 1>
    gravityOverride {z=-1500}
    physicsBodyDefinitions [{
      boundBone {boneName=tail4}
      boxExtents=<30 30 30>
      constraintSetup {angularLimitsMax=<0 30 30>, angularLimitsMin=<-30 -30 -30>, angularTarget=<1 0 0>}
      localJointOffset=<1 0 0>
     }]
   }
   n31 = "Transform (Modify) Bone - Bone: tail1"!(ComponentPose=@n30, bAlphaBoolEnabled:bool=@n27, Translation="0,0,0", Rotation="0,0,30", Scale="1,1,1") -> (Pose) {
    alphaInputType=Bool
    rotationMode=BMM_Additive
    scaleMode=BMM_Additive
    translationMode=BMM_Additive
    alphaBoolBlend {blendOption=Linear}
    boneToModify {boneName=tail1}
    rotation {roll=30}
   }
   n32 = "Transform (Modify) Bone - Bone: tail2"!(ComponentPose=@n31, bAlphaBoolEnabled:bool=@n27, Translation="0,0,0", Rotation="0,0,10", Scale="1,1,1") -> (Pose) {
    alphaInputType=Bool
    rotationMode=BMM_Additive
    scaleMode=BMM_Additive
    translationMode=BMM_Additive
    alphaBoolBlend {blendOption=Linear}
    boneToModify {boneName=tail2}
    rotation {roll=10}
   }
   n33 = "Transform (Modify) Bone - Bone: tail3"!(ComponentPose=@n32, bAlphaBoolEnabled:bool=@n27, Translation="0,0,0", Rotation="0,0,10", Scale="1,1,1") -> (Pose) {
    alphaInputType=Bool
    rotationMode=BMM_Additive
    scaleMode=BMM_Additive
    translationMode=BMM_Additive
    alphaBoolBlend {blendOption=Linear}
    boneToModify {boneName=tail3}
    rotation {roll=10}
   }
   n34 = "Transform (Modify) Bone - Bone: tail4"!(ComponentPose=@n33, bAlphaBoolEnabled:bool=@n27, Translation="0,0,0", Rotation="0,0,10", Scale="1,1,1") -> (Pose) {
    alphaInputType=Bool
    rotationMode=BMM_Additive
    scaleMode=BMM_Additive
    translationMode=BMM_Additive
    alphaBoolBlend {blendOption=Linear}
    boneToModify {boneName=tail4}
    rotation {roll=10}
   }
   n35 = "Transform (Modify) Bone - Bone: deform_upperarm_twist_1_l"!(ComponentPose=@n34, bAlphaBoolEnabled:bool=@n27, Translation="5,0,0", Rotation="0,0,0", Scale="1,1,1") -> (Pose) {
    alphaInputType=Bool
    rotationMode=BMM_Additive
    scaleMode=BMM_Additive
    translationMode=BMM_Additive
    alphaBoolBlend {blendOption=Linear}
    boneToModify {boneName=deform_upperarm_twist_1_l}
    translation {x=5}
   }
   n36 = "Transform (Modify) Bone - Bone: deform_upperarm_twist_1_r"!(ComponentPose=@n35, bAlphaBoolEnabled:bool=@n27, Translation="-6,0,0", Rotation="0,0,20", Scale="1.2,1.2,1.2") -> (Pose) {
    alphaInputType=Bool
    rotationMode=BMM_Additive
    scaleMode=BMM_Additive
    translationMode=BMM_Additive
    alphaBoolBlend {blendOption=Linear}
    boneToModify {boneName=deform_upperarm_twist_1_r}
    rotation {roll=20}
    scale=<1.2 1.2 1.2>
    translation {x=-6}
   }
   n37 = "Transform (Modify) Bone - Bone: deform_upperarm_twist_2_l"!(ComponentPose=@n36, bAlphaBoolEnabled:bool=@n27, Translation="0,0,0", Rotation="0,0,0", Scale="1,1,1") -> (Pose) {
    alphaInputType=Bool
    rotationMode=BMM_Additive
    scaleMode=BMM_Additive
    translationMode=BMM_Additive
    alphaBoolBlend {blendOption=Linear}
    boneToModify {boneName=deform_upperarm_twist_2_l}
   }
   n38 = "Transform (Modify) Bone - Bone: deform_upperarm_twist_2_r"!(ComponentPose=@n37, bAlphaBoolEnabled:bool=@n27, Translation="-3,-5,3", Rotation="0,0,0", Scale="1,1,1") -> (Pose) {
    alphaInputType=Bool
    rotationMode=BMM_Additive
    scaleMode=BMM_Additive
    translationMode=BMM_Additive
    alphaBoolBlend {blendOption=Linear}
    boneToModify {boneName=deform_upperarm_twist_2_r}
    translation=<-3 -5 3>
   }
   n39 = "Transform (Modify) Bone - Bone: deform_upperarm_twist_3_l"!(ComponentPose=@n38, bAlphaBoolEnabled:bool=@n27, Translation="0,0,0", Rotation="0,0,0", Scale="1,1,1") -> (Pose) {
    alphaInputType=Bool
    rotationMode=BMM_Additive
    scaleMode=BMM_Additive
    translationMode=BMM_Additive
    alphaBoolBlend {blendOption=Linear}
    boneToModify {boneName=deform_upperarm_twist_3_l}
   }
   n40 = "Transform (Modify) Bone - Bone: deform_upperarm_twist_3_r"!(ComponentPose=@n39, bAlphaBoolEnabled:bool=@n27, Translation="0,0,0", Rotation="0,0,0", Scale="0.8,0.8,0.8") -> (Pose) {
    alphaInputType=Bool
    rotationMode=BMM_Additive
    scaleMode=BMM_Additive
    translationMode=BMM_Additive
    alphaBoolBlend {blendOption=Linear}
    boneToModify {boneName=deform_upperarm_twist_3_r}
    scale=<0.8 0.8 0.8>
   }
   n41 = "Transform (Modify) Bone - Bone: tail1"!(ComponentPose=@n40, Alpha:float=@n2, Translation="0,0,20", Rotation="0,0,10", Scale="1,1,1") -> (Pose) {
    rotationMode=BMM_Additive
    scaleMode=BMM_Additive
    translationMode=BMM_Additive
    alphaBoolBlend {blendOption=Linear}
    alphaScaleBiasClamp {bInterpResult=true, interpSpeedDecreasing=1, interpSpeedIncreasing=1}
    boneToModify {boneName=tail1}
    rotation {roll=10}
    translation {z=20}
   }
   n42 = "Transform (Modify) Bone - Bone: deform_pelvis"(ComponentPose=@n41, Alpha=@n2, Translation="0,0,5", Rotation="0,0,5", Scale="1,1,1") {
    rotationMode=BMM_Additive
    scaleMode=BMM_Additive
    translationMode=BMM_Additive
    alphaBoolBlend {blendOption=Linear}
    alphaScaleBiasClamp {bInterpResult=true, interpSpeedDecreasing=1, interpSpeedIncreasing=1}
    boneToModify {boneName=deform_pelvis}
    rotation {roll=5}
    translation {z=5}
   }
   n43 = Get isSkyDiveDivingFloat() -> (isSkyDiveDivingFloat:double)
   n44 = "Transform (Modify) Bone - Bone: tail1"(ComponentPose=@n42, Alpha=@n43, Translation="0,0,0", Rotation="0,0,-15", Scale="1,1,1") {
    rotationMode=BMM_Additive
    scaleMode=BMM_Additive
    translationMode=BMM_Additive
    alphaBoolBlend {blendOption=Linear}
    alphaScaleBiasClamp {bInterpResult=true, interpSpeedIncreasing=0.7}
    boneToModify {boneName=tail1}
    rotation {roll=-15}
   }
   n45 = "Transform (Modify) Bone - Bone: tail2"!(ComponentPose=@n44, Alpha:float=@n43, Translation="0,0,0", Rotation="0,0,-15", Scale="1,1,1") -> (Pose) {
    rotationMode=BMM_Additive
    scaleMode=BMM_Additive
    translationMode=BMM_Additive
    alphaBoolBlend {blendOption=Linear}
    alphaScaleBiasClamp {bInterpResult=true, interpSpeedIncreasing=0.7}
    boneToModify {boneName=tail2}
    rotation {roll=-15}
   }
   n46 = "Transform (Modify) Bone - Bone: tail3"!(ComponentPose=@n45, Alpha:float=@n43, Translation="0,0,0", Rotation="0,0,-30", Scale="1,1,1") -> (Pose) {
    rotationMode=BMM_Additive
    scaleMode=BMM_Additive
    translationMode=BMM_Additive
    alphaBoolBlend {blendOption=Linear}
    alphaScaleBiasClamp {bInterpResult=true, interpSpeedIncreasing=0.7}
    boneToModify {boneName=tail3}
    rotation {roll=-30}
   }
   n47 = "Transform (Modify) Bone - Bone: tail4"!(ComponentPose=@n46, Alpha:float=@n43, Translation="0,0,0", Rotation="0,0,-30", Scale="1,1,1") -> (Pose) {
    rotationMode=BMM_Additive
    scaleMode=BMM_Additive
    translationMode=BMM_Additive
    alphaBoolBlend {blendOption=Linear}
    alphaScaleBiasClamp {bInterpResult=true, interpSpeedIncreasing=0.7}
    boneToModify {boneName=tail4}
    rotation {roll=-30}
   }
   n48 = Get Is Swim Sprinting() -> (Is Swim Sprinting:double)
   n49 = "Transform (Modify) Bone - Bone: tail1"(ComponentPose=@n47, Alpha=@n48, Translation="0,0,0", Rotation="0,0,-15", Scale="1,1,1") {
    rotationMode=BMM_Additive
    scaleMode=BMM_Additive
    translationMode=BMM_Additive
    alphaBoolBlend {blendOption=Linear}
    alphaScaleBiasClamp {bInterpResult=true, interpSpeedIncreasing=2}
    boneToModify {boneName=tail1}
    rotation {roll=-15}
   }
   n50 = "Transform (Modify) Bone - Bone: tail2"(ComponentPose=@n49, Alpha=@n48, Translation="0,0,0", Rotation="0,0,-15", Scale="1,1,1") {
    rotationMode=BMM_Additive
    scaleMode=BMM_Additive
    translationMode=BMM_Additive
    alphaBoolBlend {blendOption=Linear}
    alphaScaleBiasClamp {bInterpResult=true, interpSpeedIncreasing=2}
    boneToModify {boneName=tail2}
    rotation {roll=-15}
   }
   n51 = "Transform (Modify) Bone - Bone: tail3"(ComponentPose=@n50, Alpha=@n48, Translation="0,0,0", Rotation="0,0,-30", Scale="1,1,1") {
    rotationMode=BMM_Additive
    scaleMode=BMM_Additive
    translationMode=BMM_Additive
    alphaBoolBlend {blendOption=Linear}
    alphaScaleBiasClamp {bInterpResult=true, interpSpeedIncreasing=2}
    boneToModify {boneName=tail3}
    rotation {roll=-30}
   }
   n52 = "Transform (Modify) Bone - Bone: tail4"(ComponentPose=@n51, Alpha=@n48, Translation="0,0,0", Rotation="0,0,-30", Scale="1,1,1") {
    rotationMode=BMM_Additive
    scaleMode=BMM_Additive
    translationMode=BMM_Additive
    alphaBoolBlend {blendOption=Linear}
    alphaScaleBiasClamp {bInterpResult=true, interpSpeedIncreasing=2}
    boneToModify {boneName=tail4}
    rotation {roll=-30}
   }
   n53 = Get Is Skydive Floating Float() -> (Is Skydive Floating Float:double)
   n54 = "Transform (Modify) Bone - Bone: tail1"(ComponentPose=@n52, Alpha=@n53, Translation="0,0,0", Rotation="0,0,-15", Scale="1,1,1") {
    rotationMode=BMM_Additive
    scaleMode=BMM_Additive
    translationMode=BMM_Additive
    alphaBoolBlend {blendOption=Linear}
    alphaScaleBiasClamp {bInterpResult=true, interpSpeedDecreasing=2, interpSpeedIncreasing=1}
    boneToModify {boneName=tail1}
    rotation {roll=-15}
   }
   n55 = "Transform (Modify) Bone - Bone: tail2"(ComponentPose=@n54, Alpha=@n53, Translation="0,0,0", Rotation="0,0,-15", Scale="1,1,1") {
    rotationMode=BMM_Additive
    scaleMode=BMM_Additive
    translationMode=BMM_Additive
    alphaBoolBlend {blendOption=Linear}
    alphaScaleBiasClamp {bInterpResult=true, interpSpeedDecreasing=2, interpSpeedIncreasing=1}
    boneToModify {boneName=tail2}
    rotation {roll=-15}
   }
   n56 = "Transform (Modify) Bone - Bone: tail3"(ComponentPose=@n55, Alpha=@n53, Translation="0,0,0", Rotation="0,0,-30", Scale="1,1,1") {
    rotationMode=BMM_Additive
    scaleMode=BMM_Additive
    translationMode=BMM_Additive
    alphaBoolBlend {blendOption=Linear}
    alphaScaleBiasClamp {bInterpResult=true, interpSpeedDecreasing=2, interpSpeedIncreasing=1}
    boneToModify {boneName=tail3}
    rotation {roll=-30}
   }
   n57 = "Transform (Modify) Bone - Bone: tail4"(ComponentPose=@n56, Alpha=@n53, Translation="0,0,0", Rotation="0,0,-30", Scale="1,1,1") {
    rotationMode=BMM_Additive
    scaleMode=BMM_Additive
    translationMode=BMM_Additive
    alphaBoolBlend {blendOption=Linear}
    alphaScaleBiasClamp {bInterpResult=true, interpSpeedDecreasing=2, interpSpeedIncreasing=1}
    boneToModify {boneName=tail4}
    rotation {roll=-30}
   }
   n58 = Get Is Knee Sliding Float() -> (Is Knee Sliding Float:double)
   n59 = "Constraint - tail1 (R)"(ComponentPose=@n57, Alpha=@n58, ConstraintWeights_0=1) {
    alphaBoolBlend {blendOption=Linear}
    boneToModify {boneName=tail1}
    constraintSetup [{transformType=Rotation, targetBone {boneName=root}}]
    constraintWeights [1]
   }
   n60 = "Transform (Modify) Bone - Bone: tail1"(ComponentPose=@n59, Alpha=@n58, Translation="0,0,0", Rotation="0,20,20", Scale="1,1,1") {
    rotationMode=BMM_Additive
    scaleMode=BMM_Additive
    translationMode=BMM_Additive
    alphaBoolBlend {blendOption=Linear}
    boneToModify {boneName=tail1}
    rotation {roll=20, yaw=20}
   }
   n61 = "Transform (Modify) Bone - Bone: deform_pelvis"(ComponentPose=@n60, Alpha=@n58, Translation="0,0,1", Rotation="0,0,30", Scale="1,1,1") {
    rotationMode=BMM_Additive
    scaleMode=BMM_Additive
    translationMode=BMM_Additive
    alphaBoolBlend {blendOption=Linear}
    alphaScaleBiasClamp {bInterpResult=true}
    boneToModify {boneName=deform_pelvis}
    rotation {roll=30}
    translation {z=1}
   }
   n62 = "Transform (Modify) Bone - Bone: tail2"(ComponentPose=@n61, Alpha=@n58, Translation="0,0,0", Rotation="0,0,30", Scale="1,1,1") {
    rotationMode=BMM_Additive
    scaleMode=BMM_Additive
    translationMode=BMM_Additive
    alphaBoolBlend {blendOption=Linear}
    alphaScaleBiasClamp {bInterpResult=true}
    boneToModify {boneName=tail2}
    rotation {roll=30}
   }
   n63 = "Transform (Modify) Bone - Bone: tail3"(ComponentPose=@n62, Alpha=@n58, Translation="0,0,0", Rotation="0,0,15", Scale="1,1,1") {
    rotationMode=BMM_Additive
    scaleMode=BMM_Additive
    translationMode=BMM_Additive
    alphaBoolBlend {blendOption=Linear}
    alphaScaleBiasClamp {bInterpResult=true}
    boneToModify {boneName=tail3}
    rotation {roll=15}
   }
   n64 = "Transform (Modify) Bone - Bone: tail4"(ComponentPose=@n63, Alpha=@n58, Translation="0,0,0", Rotation="0,0,15", Scale="1,1,1") {
    rotationMode=BMM_Additive
    scaleMode=BMM_Additive
    translationMode=BMM_Additive
    alphaBoolBlend {blendOption=Linear}
    alphaScaleBiasClamp {bInterpResult=true}
    boneToModify {boneName=tail4}
    rotation {roll=15}
   }
   n65 = "Transform (Modify) Bone - Bone: tail1"(ComponentPose=@n64, Alpha=@n1, Translation="0,0,5", Rotation="0,0,60", Scale="1,1,1") {
    rotationMode=BMM_Additive
    scaleMode=BMM_Additive
    translationMode=BMM_Additive
    alphaBoolBlend {blendOption=Linear}
    alphaScaleBiasClamp {bInterpResult=true, interpSpeedIncreasing=3}
    boneToModify {boneName=tail1}
    rotation {roll=60}
    translation {z=5}
   }
   n66 = "Transform (Modify) Bone - Bone: deform_pelvis"(ComponentPose=@n65, Alpha=@n1, Translation="0,0,5", Rotation="0,0,30", Scale="1,1,1") {
    rotationMode=BMM_Additive
    scaleMode=BMM_Additive
    translationMode=BMM_Additive
    alphaBoolBlend {blendOption=Linear}
    alphaScaleBiasClamp {bInterpResult=true, interpSpeedIncreasing=3}
    boneToModify {boneName=deform_pelvis}
    rotation {roll=30}
    translation {z=5}
   }
   n67 = "Transform (Modify) Bone - Bone: tail2"(ComponentPose=@n66, Alpha=@n1, Translation="0,0,0", Rotation="0,0,30", Scale="1,1,1") {
    rotationMode=BMM_Additive
    scaleMode=BMM_Additive
    translationMode=BMM_Additive
    alphaBoolBlend {blendOption=Linear}
    alphaScaleBiasClamp {bInterpResult=true, interpSpeedIncreasing=3}
    boneToModify {boneName=tail2}
    rotation {roll=30}
   }
   n68 = "Transform (Modify) Bone - Bone: tail3"(ComponentPose=@n67, Alpha=@n1, Translation="0,0,0", Rotation="0,0,30", Scale="1,1,1") {
    rotationMode=BMM_Additive
    scaleMode=BMM_Additive
    translationMode=BMM_Additive
    alphaBoolBlend {blendOption=Linear}
    alphaScaleBiasClamp {bInterpResult=true, interpSpeedIncreasing=3}
    boneToModify {boneName=tail3}
    rotation {roll=30}
   }
   n69 = "Transform (Modify) Bone - Bone: tail4"(ComponentPose=@n68, Alpha=@n1, Translation="0,0,0", Rotation="0,0,30", Scale="1,1,1") {
    rotationMode=BMM_Additive
    scaleMode=BMM_Additive
    translationMode=BMM_Additive
    alphaBoolBlend {blendOption=Linear}
    alphaScaleBiasClamp {bInterpResult=true, interpSpeedIncreasing=3}
    boneToModify {boneName=tail4}
    rotation {roll=30}
   }
   n70 = Get isWearingBackbling()
   n71 = "Transform (Modify) Bone - Bone: spike_01"!(ComponentPose=@n69, bAlphaBoolEnabled:bool=@n70, Translation="0,0,0", Rotation="0,0,0", Scale="0.5,0.2,0.2") -> (Pose) {
    alphaInputType=Bool
    rotationMode=BMM_Additive
    scaleMode=BMM_Additive
    translationMode=BMM_Additive
    alphaBoolBlend {blendOption=Linear}
    alphaScaleBiasClamp {bInterpResult=true}
    boneToModify {boneName=spike_01}
    scale=<0.5 0.2 0.2>
   }
   n72 = "Transform (Modify) Bone - Bone: spike_01_l"!(ComponentPose=@n71, bAlphaBoolEnabled:bool=@n70, Translation="0,0,0", Rotation="0,0,0", Scale="0.5,0.2,0.2") -> (Pose) {
    alphaInputType=Bool
    rotationMode=BMM_Additive
    scaleMode=BMM_Additive
    translationMode=BMM_Additive
    alphaBoolBlend {blendOption=Linear}
    alphaScaleBiasClamp {bInterpResult=true}
    boneToModify {boneName=spike_01_l}
    scale=<0.5 0.2 0.2>
   }
   n73 = "Transform (Modify) Bone - Bone: spike_01_r"!(ComponentPose=@n72, bAlphaBoolEnabled:bool=@n70, Translation="0,0,0", Rotation="0,0,0", Scale="0.5,0.2,0.2") -> (Pose) {
    alphaInputType=Bool
    rotationMode=BMM_Additive
    scaleMode=BMM_Additive
    translationMode=BMM_Additive
    alphaBoolBlend {blendOption=Linear}
    alphaScaleBiasClamp {bInterpResult=true}
    boneToModify {boneName=spike_01_r}
    scale=<0.5 0.2 0.2>
   }
   n74 = "Transform (Modify) Bone - Bone: spike_02"!(ComponentPose=@n73, bAlphaBoolEnabled:bool=@n70, Translation="0,0,0", Rotation="0,0,0", Scale="0.2,0.2,0.2") -> (Pose) {
    alphaInputType=Bool
    rotationMode=BMM_Additive
    scaleMode=BMM_Additive
    translationMode=BMM_Additive
    alphaBoolBlend {blendOption=Linear}
    alphaScaleBiasClamp {bInterpResult=true}
    boneToModify {boneName=spike_02}
    scale=<0.2 0.2 0.2>
   }
   n75 = "Transform (Modify) Bone - Bone: spike_02_l"!(ComponentPose=@n74, bAlphaBoolEnabled:bool=@n70, Translation="0,0,0", Rotation="0,0,0", Scale="0.2,0.2,0.2") -> (Pose) {
    alphaInputType=Bool
    rotationMode=BMM_Additive
    scaleMode=BMM_Additive
    translationMode=BMM_Additive
    alphaBoolBlend {blendOption=Linear}
    alphaScaleBiasClamp {bInterpResult=true}
    boneToModify {boneName=spike_02_l}
    scale=<0.2 0.2 0.2>
   }
   n76 = "Transform (Modify) Bone - Bone: spike_02_r"!(ComponentPose=@n75, bAlphaBoolEnabled:bool=@n70, Translation="0,0,0", Rotation="0,0,0", Scale="0.2,0.2,0.2") -> (Pose) {
    alphaInputType=Bool
    rotationMode=BMM_Additive
    scaleMode=BMM_Additive
    translationMode=BMM_Additive
    alphaBoolBlend {blendOption=Linear}
    alphaScaleBiasClamp {bInterpResult=true}
    boneToModify {boneName=spike_02_r}
    scale=<0.2 0.2 0.2>
   }
   n77 = "Transform (Modify) Bone - Bone: spike_03"!(ComponentPose=@n76, bAlphaBoolEnabled:bool=@n70, Translation="0,0,0", Rotation="0,0,0", Scale="0.2,0.2,0.2") -> (Pose) {
    alphaInputType=Bool
    rotationMode=BMM_Additive
    scaleMode=BMM_Additive
    translationMode=BMM_Additive
    alphaBoolBlend {blendOption=Linear}
    alphaScaleBiasClamp {bInterpResult=true}
    boneToModify {boneName=spike_03}
    scale=<0.2 0.2 0.2>
   }
   n78 = "Transform (Modify) Bone - Bone: spike_03_l"!(ComponentPose=@n77, bAlphaBoolEnabled:bool=@n70, Translation="0,0,0", Rotation="0,0,0", Scale="0.2,0.2,0.2") -> (Pose) {
    alphaInputType=Bool
    rotationMode=BMM_Additive
    scaleMode=BMM_Additive
    translationMode=BMM_Additive
    alphaBoolBlend {blendOption=Linear}
    alphaScaleBiasClamp {bInterpResult=true}
    boneToModify {boneName=spike_03_l}
    scale=<0.2 0.2 0.2>
   }
   n79 = "Transform (Modify) Bone - Bone: spike_03_r"!(ComponentPose=@n78, bAlphaBoolEnabled:bool=@n70, Translation="0,0,0", Rotation="0,0,0", Scale="0.2,0.2,0.2") -> (Pose) {
    alphaInputType=Bool
    rotationMode=BMM_Additive
    scaleMode=BMM_Additive
    translationMode=BMM_Additive
    alphaBoolBlend {blendOption=Linear}
    alphaScaleBiasClamp {bInterpResult=true}
    boneToModify {boneName=spike_03_r}
    scale=<0.2 0.2 0.2>
   }
   n80 = Get isWearingCustomBackpack() -> (isWearingCustomBackpack:bool)
   n81 = "Transform (Modify) Bone - Bone: spike_01"(ComponentPose=@n79, bAlphaBoolEnabled=@n80, Translation="0,0,0", Rotation="0,0,0", Scale="2,5,5") {
    alphaInputType=Bool
    rotationMode=BMM_Additive
    scaleMode=BMM_Additive
    translationMode=BMM_Additive
    alphaBoolBlend {blendOption=Linear}
    alphaScaleBiasClamp {bInterpResult=true}
    boneToModify {boneName=spike_01}
    scale=<2 5 5>
   }
   n82 = "Transform (Modify) Bone - Bone: spike_01_l"(ComponentPose=@n81, bAlphaBoolEnabled=@n80, Translation="0,0,0", Rotation="0,0,0", Scale="2,5,5") {
    alphaInputType=Bool
    rotationMode=BMM_Additive
    scaleMode=BMM_Additive
    translationMode=BMM_Additive
    alphaBoolBlend {blendOption=Linear}
    alphaScaleBiasClamp {bInterpResult=true}
    boneToModify {boneName=spike_01_l}
    scale=<2 5 5>
   }
   n83 = "Transform (Modify) Bone - Bone: spike_01_r"(ComponentPose=@n82, bAlphaBoolEnabled=@n80, Translation="0,0,0", Rotation="0,0,0", Scale="2,5,5") {
    alphaInputType=Bool
    rotationMode=BMM_Additive
    scaleMode=BMM_Additive
    translationMode=BMM_Additive
    alphaBoolBlend {blendOption=Linear}
    alphaScaleBiasClamp {bInterpResult=true}
    boneToModify {boneName=spike_01_r}
    scale=<2 5 5>
   }
   n84 = "Transform (Modify) Bone - Bone: spike_02"(ComponentPose=@n83, bAlphaBoolEnabled=@n80, Translation="0,0,0", Rotation="0,0,0", Scale="5,5,5") {
    alphaInputType=Bool
    rotationMode=BMM_Additive
    scaleMode=BMM_Additive
    translationMode=BMM_Additive
    alphaBoolBlend {blendOption=Linear}
    alphaScaleBiasClamp {bInterpResult=true}
    boneToModify {boneName=spike_02}
    scale=<5 5 5>
   }
   n85 = "Transform (Modify) Bone - Bone: spike_02_l"(ComponentPose=@n84, bAlphaBoolEnabled=@n80, Translation="0,0,0", Rotation="0,0,0", Scale="5,5,5") {
    alphaInputType=Bool
    rotationMode=BMM_Additive
    scaleMode=BMM_Additive
    translationMode=BMM_Additive
    alphaBoolBlend {blendOption=Linear}
    alphaScaleBiasClamp {bInterpResult=true}
    boneToModify {boneName=spike_02_l}
    scale=<5 5 5>
   }
   n86 = "Transform (Modify) Bone - Bone: spike_02_r"(ComponentPose=@n85, bAlphaBoolEnabled=@n80, Translation="0,0,0", Rotation="0,0,0", Scale="5,5,5") {
    alphaInputType=Bool
    rotationMode=BMM_Additive
    scaleMode=BMM_Additive
    translationMode=BMM_Additive
    alphaBoolBlend {blendOption=Linear}
    alphaScaleBiasClamp {bInterpResult=true}
    boneToModify {boneName=spike_02_r}
    scale=<5 5 5>
   }
   n87 = "Transform (Modify) Bone - Bone: spike_03"(ComponentPose=@n86, bAlphaBoolEnabled=@n80, Translation="0,0,0", Rotation="0,0,0", Scale="5,5,5") {
    alphaInputType=Bool
    rotationMode=BMM_Additive
    scaleMode=BMM_Additive
    translationMode=BMM_Additive
    alphaBoolBlend {blendOption=Linear}
    alphaScaleBiasClamp {bInterpResult=true}
    boneToModify {boneName=spike_03}
    scale=<5 5 5>
   }
   n88 = "Transform (Modify) Bone - Bone: spike_03_l"(ComponentPose=@n87, bAlphaBoolEnabled=@n80, Translation="0,0,0", Rotation="0,0,0", Scale="5,5,5") {
    alphaInputType=Bool
    rotationMode=BMM_Additive
    scaleMode=BMM_Additive
    translationMode=BMM_Additive
    alphaBoolBlend {blendOption=Linear}
    alphaScaleBiasClamp {bInterpResult=true}
    boneToModify {boneName=spike_03_l}
    scale=<5 5 5>
   }
   n89 = "Transform (Modify) Bone - Bone: spike_03_r"(ComponentPose=@n88, bAlphaBoolEnabled=@n80, Translation="0,0,0", Rotation="0,0,0", Scale="5,5,5") {
    alphaInputType=Bool
    rotationMode=BMM_Additive
    scaleMode=BMM_Additive
    translationMode=BMM_Additive
    alphaBoolBlend {blendOption=Linear}
    alphaScaleBiasClamp {bInterpResult=true}
    boneToModify {boneName=spike_03_r}
    scale=<5 5 5>
   }
   n90 = Get Frontend_Backpack_Pose() -> (Frontend_Backpack_Pose:float)
   n91 = "Transform (Modify) Bone - Bone: spine_03"(
    ComponentPose="Transform (Modify) Bone - Bone: spine_04"(
     ComponentPose="Transform (Modify) Bone - Bone: spine_05"(ComponentPose=@n89, Alpha=@n90, Translation="0,0,0", Rotation="0,0,0", Scale="1,1,1") {rotationMode=BMM_Additive, alphaBoolBlend {blendOption=Linear}, boneToModify {boneName=spine_05}}
     Alpha=@n90
     Translation="0,0,0"
     Rotation="0,0,0"
     Scale="1,1,1"
    ) {rotationMode=BMM_Additive, alphaBoolBlend {blendOption=Linear}, boneToModify {boneName=spine_04}}
    Alpha=@n90
    Translation="0,0,0"
    Rotation="0,0,10"
    Scale="1,1,1"
   ) {rotationMode=BMM_Additive, alphaBoolBlend {blendOption=Linear}, boneToModify {boneName=spine_03}, rotation {roll=10}}
   n92 = "Transform (Modify) Bone - Bone: spine_02"(ComponentPose=@n91, Alpha=@n90, Translation="0,0,0", Rotation="0,0,10", Scale="1,1,1") {rotationMode=BMM_Additive, alphaBoolBlend {blendOption=Linear}, boneToModify {boneName=spine_02}, rotation {roll=10}}
   n93 = "Transform (Modify) Bone - Bone: spine_01"(ComponentPose=@n92, Alpha=@n90, Translation="0,0,0", Rotation="10,0,10", Scale="1,1,1") {rotationMode=BMM_Additive, alphaBoolBlend {blendOption=Linear}, boneToModify {boneName=spine_01}, rotation {pitch=10, roll=10}}
   n94 = "Transform (Modify) Bone - Bone: upperarm_l"(ComponentPose=@n93, Alpha=@n90, Translation="0,0,0", Rotation="-4.659,0.000,-21.511", Scale="1,1,1") {rotationMode=BMM_Additive, alphaBoolBlend {blendOption=Linear}, boneToModify {boneName=upperarm_l}, rotation {pitch=-4.659, roll=-21.511}}
   n95 = "Transform (Modify) Bone - Bone: upperarm_r"(ComponentPose=@n94, Alpha=@n90, Translation="0,0,0", Rotation="-5.306,0.915,-21.778", Scale="1,1,1") {rotationMode=BMM_Additive, alphaBoolBlend {blendOption=Linear}, boneToModify {boneName=upperarm_r}, rotation=<rot -5.306 0.915 -21.778>}
   n96 = "Transform (Modify) Bone - Bone: head"(ComponentPose=@n95, Alpha=@n90, Translation="0,0,0", Rotation="0,0,-22", Scale="1,1,1") {rotationMode=BMM_Additive, alphaBoolBlend {blendOption=Linear}, boneToModify {boneName=head}, rotation {roll=-22}}
   n97 = "Transform (Modify) Bone - Bone: neck_01"(ComponentPose=@n96, Alpha=@n90, Translation="0,0,0", Rotation="-5.113,21.316,-9.715", Scale="1,1,1") {rotationMode=BMM_Additive, alphaBoolBlend {blendOption=Linear}, boneToModify {boneName=neck_01}, rotation=<rot -5.113 21.316 -9.715>}
   n98 = Layered blend per bone(
    BasePose=Component To Local(ComponentPose=@n97)
    BlendPoses_0=Slot 'DefaultSlot'()
    BlendWeights_0=Get Custom Idle Is Playing() -> (Custom Idle Is Playing:double)
   ) {layerSetup [{branchFilters [{boneName=tail1}, {boneName=tail2}, {boneName=tail3}, {boneName=tail4}]}]}
   Output Pose(Result=@n98) {layerGroup=None}
   n99 = Control Rig(Source=Copy Pose From Mesh() {bCopyCurves=true, bUseAttachedParent=true}) {
    alphaBoolBlend {blendOption=Linear}
    controlRigAssetReference {blueprintRigClass=&/M_MED_BlondeJaw_Strike_CtrlRig_DUP.M_MED_BlondeJaw_Strike_CtrlRig_DUP_C}
   }
   n100 = Get tail_cape_begin() -> (tail_cape_begin:float)
   n101 = "Transform (Modify) Bone - Bone: tail1"(ComponentPose=Local To Component(LocalPose=@n99), Alpha=@n100, Translation="0,0,0", Rotation="0,0,-20", Scale="1,1,1") {
    rotationMode=BMM_Additive
    alphaBoolBlend {blendOption=Linear}
    alphaScaleBiasClamp {bInterpResult=true}
    boneToModify {boneName=tail1}
    rotation {roll=-20}
   }
   n102 = "Transform (Modify) Bone - Bone: tail2"(ComponentPose=@n101, Alpha=@n100, Translation="0,0,0", Rotation="0,0,-10", Scale="1,1,1") {
    rotationMode=BMM_Additive
    alphaBoolBlend {blendOption=Linear}
    alphaScaleBiasClamp {bInterpResult=true}
    boneToModify {boneName=tail2}
    rotation {roll=-10}
   }
   n103 = "Transform (Modify) Bone - Bone: tail3"(ComponentPose=@n102, Alpha=@n100, Translation="0,0,0", Rotation="0,0,-5", Scale="1,1,1") {
    rotationMode=BMM_Additive
    alphaBoolBlend {blendOption=Linear}
    alphaScaleBiasClamp {bInterpResult=true}
    boneToModify {boneName=tail3}
    rotation {roll=-5}
   }
   n104 = "Transform (Modify) Bone - Bone: tail1"(ComponentPose=@n103, Alpha=1, Translation="0,0,0", Rotation="0,0,-5", Scale="1,1,1") {
    rotationMode=BMM_Additive
    alphaBoolBlend {blendOption=Linear}
    alphaScaleBiasClamp {bInterpResult=true, interpSpeedDecreasing=4, interpSpeedIncreasing=4}
    boneToModify {boneName=tail1}
    rotation {roll=-5}
   }
   n105 = "Transform (Modify) Bone - Bone: tail2"(ComponentPose=@n104, Alpha=1, Translation="0,0,0", Rotation="0,0,-5", Scale="1,1,1") {
    rotationMode=BMM_Additive
    alphaBoolBlend {blendOption=Linear}
    alphaScaleBiasClamp {bInterpResult=true, interpSpeedDecreasing=4, interpSpeedIncreasing=4}
    boneToModify {boneName=tail2}
    rotation {roll=-5}
   }
   n106 = "Transform (Modify) Bone - Bone: tail3"(ComponentPose=@n105, Alpha=1, Translation="0,0,0", Rotation="0,0,-5", Scale="1,1,1") {
    rotationMode=BMM_Additive
    alphaBoolBlend {blendOption=Linear}
    alphaScaleBiasClamp {bInterpResult=true, interpSpeedDecreasing=4, interpSpeedIncreasing=4}
    boneToModify {boneName=tail3}
    rotation {roll=-5}
   }
   n107 = "Transform (Modify) Bone - Bone: tail1"(ComponentPose=@n106, Alpha=1, Translation="0,0,0", Rotation="0,0,-5", Scale="1,1,1") {
    rotationMode=BMM_Additive
    alphaBoolBlend {blendOption=Linear}
    alphaScaleBiasClamp {bInterpResult=true, interpSpeedDecreasing=4, interpSpeedIncreasing=4}
    boneToModify {boneName=tail1}
    rotation {roll=-5}
   }
   n108 = "Transform (Modify) Bone - Bone: tail2"(ComponentPose=@n107, Alpha=1, Translation="0,0,0", Rotation="0,0,-5", Scale="1,1,1") {
    rotationMode=BMM_Additive
    alphaBoolBlend {blendOption=Linear}
    alphaScaleBiasClamp {bInterpResult=true, interpSpeedDecreasing=4, interpSpeedIncreasing=4}
    boneToModify {boneName=tail2}
    rotation {roll=-5}
   }
   n109 = "Transform (Modify) Bone - Bone: tail3"(ComponentPose=@n108, Alpha=1, Translation="0,0,0", Rotation="0,0,5", Scale="1,1,1") {
    rotationMode=BMM_Additive
    alphaBoolBlend {blendOption=Linear}
    alphaScaleBiasClamp {bInterpResult=true, interpSpeedDecreasing=4, interpSpeedIncreasing=4}
    boneToModify {boneName=tail3}
    rotation {roll=5}
   }
   Save cached pose 'Presim'(Pose=Component To Local(ComponentPose=@n109)) {cacheName=Presim}
  ]}
 EventGraph {nodes [
   Event BlueprintUpdateAnimation()
   > n1 = Sequence!(execute) -> (
    then_0
    then_1
    then_2:Exec
    then_3:Exec
    then_4:Exec
    then_5:Exec
    then_6:Exec
    then_7:Exec
    then_8:Exec
    then_9:Exec
    then_10:Exec
    then_11:Exec
   )
   > Set Is Knee Sliding Float(Is Knee Sliding Float:double="To Float (Boolean)"(InBool=Get bIsKneeSliding())) -> (Output_Get:double)
   > Set Is Baseball Sliding float(Is Baseball Sliding float:double="To Float (Boolean)"(InBool=Get bIsBaseballSliding())) -> (Output_Get:double)
   Set Is Tactical Sprinting(execute=@n1.then_1, Is Tactical Sprinting:double="To Float (Boolean)"(InBool=Get bIsTacticalSprinting())) -> (Output_Get:double)
   n2 = Sequence!(execute=@n1.then_10) -> (then_0, then_1)
   > n3 = Branch(Condition=Get bIsPlayingEmote())
   n4 = GetAnimInstance(self=GetOwnerSkeletalMeshComponent())
   > Is Valid(InputObject=@n4)
   n5 = OR Boolean(
    A=Montage_IsPlaying(self=@n4, Montage=/BRCosmetics/Animation/Game/MainPlayer/Emotes/Scrub_A_Dub/Emote_ScrubaDub_CMM_M)
    B=Montage_IsPlaying(self=@n4, Montage=/BRCosmetics/Animation/Game/MainPlayer/Emotes/Cauldron/Emote_Cauldron_CMM_M)
   )
   > Set in vehicle scale alpha(in vehicle scale alpha:double=SelectFloat(A=1, B=0, bPickA=@n5)) -> (Output_Get:double)
   > Set In Vehicle Sim Alpha(In Vehicle Sim Alpha:double=SelectFloat(A=0, B=1, bPickA=@n5)) -> (Output_Get:double)
   Set in vehicle scale alpha(execute=@n3.else, in vehicle scale alpha=0)
   > Set In Vehicle Sim Alpha(In Vehicle Sim Alpha=1)
   n6 = GetCurrentWeapon!(self:Fort Pawn Object Reference=TryGetFortPawn()) -> (ReturnValue:Fort Weapon Object Reference)
   n7 = Branch(execute=@n2.then_1, Condition=IsCosmeticObjectDisplayedAsPreview!(CosmeticObject:Object Reference=@n6) -> (ReturnValue:bool))
   > Set Frontend Pickaxe Alpha(Frontend Pickaxe Alpha:double=1) -> (Output_Get:double)
   Set Frontend Pickaxe Alpha(execute=@n7.else, Frontend Pickaxe Alpha=0)
   Set isSkyDiveDivingFloat(execute=@n1.then_2, isSkyDiveDivingFloat:double="To Float (Boolean)"(InBool=Get bIsSkydiveDiveMode())) -> (Output_Get:double)
   > Set Is Skydive Floating Float(Is Skydive Floating Float:double="To Float (Boolean)"(InBool=Get bIsSkydiveFloating())) -> (Output_Get:double)
   Set Is Surface Swimming(execute=@n1.then_3, Is Surface Swimming:double="To Float (Boolean)"(InBool=Get bIsSurfaceSwimming())) -> (Output_Get:double)
   > Set Is Swim Sprinting(Is Swim Sprinting:double="To Float (Boolean)"(InBool=Get bIsSwimSprinting() -> (bIsSwimSprinting:bool))) -> (Output_Get:double)
   n8 = NOT Boolean(A=Get bIsPlayingEmote())
   n9 = Get bIsInFrontEnd()
   Set frontendNoEmote(execute=@n1.then_4, frontendNoEmote:double="To Float (Boolean)"(InBool=AND Boolean(A=@n9, B=@n8))) -> (Output_Get:double)
   n10 = Set pawnVelocitySpeed(execute=@n1.then_5, pawnVelocitySpeed:double=Vector Length(A=Get PawnVelocity())) -> (Output_Get:double)
   n11 = Get bIsSkyDiving()
   n12 = SelectFloat(A=0, B=1, bPickA=AND Boolean(A=NOT Boolean(A=@n11), B="float > float"(A=@n10, B=2000) {is=title K2Node_CallFunction}))
   > n13 = Set pawnVelocitySpeedCutoff(pawnVelocitySpeedCutoff:double=@n12) -> (Output_Get:double)
   n14 = Get bIsPlayingEmote()
   n15 = Branch(execute=@n1.then_6, Condition=AND Boolean(A=Get bIsInFrontEnd(), B=NOT Boolean(A=@n14)))
   > Set animDynAlpha(animDynAlpha:double=0) -> (Output_Get:double)
   n16 = Branch(execute=@n15.else, Condition=Get bIsSliding())
   > Set animDynAlpha(animDynAlpha=float * float(A=@n13, B=0.1) {is=title K2Node_PromotableOperator})
   n17 = Branch(execute=@n16.else, Condition=Get IsHangGlider())
   > Set animDynAlpha(animDynAlpha=float * float(A=@n13, B=0.1))
   Set animDynAlpha(execute=@n17.else, animDynAlpha=float * float(A=@n12, B=1))
   n18 = NOT Boolean(A=Get bIsPlayingEmote())
   n19 = CheckForBackbling(LongBackblingThreshold=20)
   n20 = OR Boolean(A=@n19.bIsBackblingACape, B=@n19.bIsBackblingLong)
   n21 = Get bIsInFrontEnd()
   Set FrontEnd_Tail Alpha(
    execute=[@n1.then_7, @n24.then_2]
    FrontEnd_Tail Alpha:float="To Float (Boolean)"(InBool=AND Boolean!(A=@n18, B=@n21, C:bool=@n20) -> (ReturnValue))
   ) -> (Output_Get:float)
   Branch(execute=@n1.then_9, Condition=Get bIsInFrontEnd())
   > Set Frontend_Backpack_Pose(Frontend_Backpack_Pose:float=GetCurveValue(CurveName=Frontend_Backpack_Pose)) -> (Output_Get:float)
   Branch(execute=@n1.then_8, Condition=Get bIsInFrontEnd())
   > Set Custom Idle Is Playing(Custom Idle Is Playing:double=GetCurveValue(self=GetOwnerFrontEndAnimInstance(), CurveName=Frontend_Idle_Bypass)) -> (Output_Get:double)
   Event OnParachuteOpened()
   > n22 = Switch on EGliderType(Selection=Get PlayerGliderType() -> (PlayerGliderType=HangGlider))
   Set IsHangGlider(execute=[@n22.Cape, @n22.NoGlider, @n22.Surfing, @n22.SurfingSimple, @n22.Umbrella], IsHangGlider:bool=false) -> (Output_Get:bool)
   Set IsHangGlider(execute=@n22.HangGlider, IsHangGlider=1)
   n23 = Event OnCharacterCustomizationLoaded()
   > n24 = Sequence!(execute) -> (then_0, then_1, then_2:Exec)
   > Set BodyMesh(BodyMesh=GetOwnerSkeletalMeshComponent())
   > Cast To PlayerPawn_Athena(Object=TryGetPawnOwner())
   n25 = CheckForBackbling()
   > Set isWearingBackbling(isWearingBackbling=@n25)
   > Set isWearingCape(isWearingCape=@n25.bIsBackblingLong)
   > Set isWearingCustomBackpack(isWearingCustomBackpack:bool="Equal (Object)"(
     A="Get (a copy)"(Array:Array of Custom Character Part Object References=@n23.CustomParts, Dimension 1=2) -> (Output:Custom Character Part Object Reference)
     B=/BRCosmetics/Characters/CharacterParts/Backpacks/CP_Backpack_BlondeJaw
    )) -> (Output_Get:bool)
   n26 = CheckForBackbling(LongBackblingThreshold=20)
   n27 = Branch(execute=@n24.then_1, Condition=OR Boolean(A=@n26, B=@n26.bIsBackblingLong))
   > Set tail_cape_begin(tail_cape_begin:float=1) -> (Output_Get:float)
   > Delay(Duration=0.1)
   > Set tail_cape_begin(tail_cape_begin=0)
   Set tail_cape_begin(execute=@n27.else, tail_cape_begin=0)
   Event OnParachuteClosed()
   > Set IsHangGlider(IsHangGlider=false)
   AND Boolean!(
    A="Equal (Enum)"(
     A:EFortWeaponCoreAnimation Enum=GetCoreAnimation!(self:Fort Weapon Object Reference=@n6) -> (ReturnValue:EFortWeaponCoreAnimation Enum)
     B:EFortWeaponCoreAnimation Enum=Melee
    )
    B=Get bIsInFrontEnd()
   ) -> (ReturnValue)
  ]}
]
variables [
 animDynAlpha {type=number}
 BodyMesh {type=SkeletalMeshComponent}
 Custom Idle Is Playing {type=number}
 Frontend Pickaxe Alpha {type=number, default=1}
 Frontend_Backpack_Pose {type=number}
 FrontEnd_Tail Alpha {type=number}
 frontendNoEmote {type=number}
 in vehicle scale alpha {type=number}
 In Vehicle Sim Alpha {type=number}
 Is Baseball Sliding float {type=number}
 Is Knee Sliding Float {type=number}
 Is Skydive Floating Float {type=number}
 Is Surface Swimming {type=number}
 Is Swim Sprinting {type=number}
 Is Tactical Sprinting {type=number}
 IsHangGlider {type=boolean}
 isSkyDiveDivingFloat {type=number}
 isWearingBackbling {type=boolean}
 isWearingCape {type=boolean}
 isWearingCustomBackpack {type=boolean}
 pawnVelocitySpeed {type=number}
 pawnVelocitySpeedCutoff {type=number}
 tail_cape_begin {type=number}
]
